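# Post-copy AnimBP for the female AtticDrill_Fin cosmetic body, adding secondary motion to sleeves and haori.
# It copies the parent pose, layers face-mesh hair bones, then runs Control Rig, attach-bone offsets, and chained spring-constrained AnimDynamics via cached poses.
# The event graph feeds movement, jump, pickaxe and emote state to the rig and toggles LOD morphs.
# Its computed dynamics alphas and gravity override are unused by the dynamics nodes, which use constants.
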AnimBlueprint /BRCosmetics/Characters/Player/Female/Medium/Bodies/F_MED_AtticDrill_Fin/Meshes/F_MED_AtticDrill_Fin_AnimBP
target_skeleton=&/F_MED_AtticDrill_Fin_Skeleton
graphs [
 AnimGraph {nodes [
   Output Pose(Result=Use cached pose 'Haori Dyn'()) {layerGroup=None}
   n1 = Layered blend per bone(
    BasePose=Copy Pose From Mesh() {bUseAttachedParent=true}
    BlendPoses_0=Copy Pose From Mesh(SourceMeshComponent="Get Face Mesh "() -> ("Face Mesh ":Skeletal Mesh Component Object Reference)) {bUseAttachedParent=true}
   ) {layerSetup [{branchFilters [
       {boneName=aim_hair_lf}
       {boneName=dyn_hair_06_lf}
       {boneName=dyn_hair_07_lf}
       {boneName=aim_hair_cr}
       {boneName=dyn_hair_06_cr}
       {boneName=dyn_hair_07_cr}
       {boneName=aim_hair_rt}
       {boneName=dyn_hair_06_rt}
       {boneName=dyn_hair_07_rt}
      ]}]}
   n2 = Control Rig!(
    Source=@n1
    Alpha
    pub isSprinting:bool=Get bIsSprinting()
    pub isKneeSliding:bool=Get bIsKneeSliding()
    pub isparachuteOpened:bool=Get bIsParachuteOpened()
    pub isCrouching:bool=Get bIsCrouching()
    pubisBaseballSliding:bool=Get bIsBaseballSliding()
    pub isSkydive:bool=Get bIsSkyDiving()
    pub isJumping:bool=Get isJumping() -> (isJumping:bool)
    pub isPlayingEasternBloc:bool=Get IsEasterBloc() -> (IsEasterBloc:bool)
    pub moveForward:double=Get moveForward() -> (moveForward:double)
    pub moveBackward:double=Get moveBackward() -> (moveBackward:double)
    pub moveLeft:double=Get strafeLeft float() -> (strafeLeft float:double)
    pub moveRight:double=Get strafeRight float() -> (strafeRight float:double)
    pub isTacticSprinting:bool=Get bIsTacticalSprinting()
    Skydive_floating:bool=Get bIsSkydiveFloating()
    pub YouBoreMe:bool=Get IsYouBoreme() -> (IsYouBoreme:bool)
    pub isPASwing:bool=Get IsPASwing() -> (IsPASwing:bool)
    pub isGrandRail:bool=Get bIsOnGrindRail()
    isMoving2D:bool=Get bIsMoving2D()
    pub isFE:bool=Get bIsInFrontEnd()
   ) -> (Pose) {controlRigAssetReference {blueprintRigClass=&/F_MED_AtticDrill_Fin_CtrlRig.F_MED_AtticDrill_Fin_CtrlRig_C}}
   Save cached pose 'ControlRig'(Pose=@n2) {cacheName=ControlRig}
   n3 = Anim Dynamics - Bone: dyn_L_sleeveHole!(
    ComponentPose=Local To Component(LocalPose=Use cached pose 'Predynamics'())
    Alpha=1
    GravityOverride:Vector="0,0,-980"
    ExternalForce="0,0,0"
   ) -> (Pose) {
    angularDampingOverride=1
    angularSpringConstant=7000
    bAngularSpring=true
    bOverrideAngularDamping=true
    bUseGravityOverride=true
    gravityScale=0.35
    simulationSpace=RootRelative
    boundBone {boneName=dyn_L_sleeveHole}
    componentAppliedLinearAccClamp=<10000 10000 10000>
    componentLinearAccScale=<1 1 1>
    componentLinearVelScale=<1 1 1>
    gravityOverride {z=-980}
    physicsBodyDefinitions [{
      boundBone {boneName=dyn_L_sleeveHole}
      boxExtents=<30 40 40>
      constraintSetup {angularLimitsMax=<5 0 5>, angularLimitsMin=<-5 -5 -5>, angularTarget=<1 0 0>}
      localJointOffset=<3 0 0>
     }]
   }
   n4 = Anim Dynamics - Bone: dyn_L_sleeveBig_1!(ComponentPose=@n3, Alpha=1, GravityOverride:Vector="0,0,-980", ExternalForce="0,0,0") -> (Pose) {
    angularDampingOverride=1
    angularSpringConstant=1000
    bAngularSpring=true
    bOverrideAngularDamping=true
    bUseGravityOverride=true
    gravityScale=0.35
    simulationSpace=RootRelative
    boundBone {boneName=dyn_L_sleeveBig_1}
    componentAppliedLinearAccClamp=<10000 10000 10000>
    componentLinearAccScale=<1 1 1>
    componentLinearVelScale=<1 1 1>
    gravityOverride {z=-980}
    physicsBodyDefinitions [{
      boundBone {boneName=dyn_L_sleeveBig_1}
      boxExtents=<30 40 40>
      constraintSetup {angularLimitsMax=<15 80 40>, angularLimitsMin=<-15 -80 -40>, angularTarget=<1 0 0>}
      localJointOffset=<4 0 0>
     }]
   }
   n5 = Anim Dynamics - Bone: dyn_L_sleeveBig_2!(ComponentPose=@n4, Alpha=1, GravityOverride:Vector="0,0,-980", ExternalForce="0,0,0") -> (Pose) {
    angularDampingOverride=1
    angularSpringConstant=1000
    bAngularSpring=true
    bOverrideAngularDamping=true
    bUseGravityOverride=true
    gravityScale=0.35
    simulationSpace=RootRelative
    boundBone {boneName=dyn_L_sleeveBig_2}
    componentAppliedLinearAccClamp=<10000 10000 10000>
    componentLinearAccScale=<1 1 1>
    componentLinearVelScale=<1 1 1>
    gravityOverride {z=-980}
    physicsBodyDefinitions [{
      boundBone {boneName=dyn_L_sleeveBig_2}
      boxExtents=<30 40 40>
      constraintSetup {angularLimitsMax=<15 80 40>, angularLimitsMin=<-15 -80 -40>, angularTarget=<1 0 0>}
      localJointOffset=<4 0 0>
     }]
   }
   n6 = Anim Dynamics - Bone: dyn_L_sleeveSmall_1!(ComponentPose=@n5, Alpha=1, GravityOverride:Vector="0,0,-980", ExternalForce="0,0,0") -> (Pose) {
    angularDampingOverride=1
    angularSpringConstant=1000
    bAngularSpring=true
    bOverrideAngularDamping=true
    bUseGravityOverride=true
    gravityScale=0.35
    simulationSpace=RootRelative
    boundBone {boneName=dyn_L_sleeveSmall_1}
    componentAppliedLinearAccClamp=<10000 10000 10000>
    componentLinearAccScale=<1 1 1>
    componentLinearVelScale=<1 1 1>
    gravityOverride {z=-980}
    physicsBodyDefinitions [{
      boundBone {boneName=dyn_L_sleeveSmall_1}
      boxExtents=<30 40 40>
      constraintSetup {angularLimitsMax=<15 80 40>, angularLimitsMin=<-15 -80 -40>, angularTarget=<1 0 0>}
      localJointOffset=<4 0 0>
     }]
   }
   n7 = Anim Dynamics - Bone: dyn_L_sleeveSmall_2!(ComponentPose=@n6, Alpha=1, GravityOverride:Vector="0,0,-980", ExternalForce="0,0,0") -> (Pose) {
    angularDampingOverride=1
    angularSpringConstant=1000
    bAngularSpring=true
    bOverrideAngularDamping=true
    bUseGravityOverride=true
    gravityScale=0.35
    simulationSpace=RootRelative
    boundBone {boneName=dyn_L_sleeveSmall_2}
    componentAppliedLinearAccClamp=<10000 10000 10000>
    componentLinearAccScale=<1 1 1>
    componentLinearVelScale=<1 1 1>
    gravityOverride {z=-980}
    physicsBodyDefinitions [{
      boundBone {boneName=dyn_L_sleeveSmall_2}
      boxExtents=<30 40 40>
      constraintSetup {angularLimitsMax=<15 80 40>, angularLimitsMin=<-15 -80 -40>, angularTarget=<1 0 0>}
      localJointOffset=<4 0 0>
     }]
   }
   n8 = Anim Dynamics - Bone: dyn_R_sleeveHole!(ComponentPose=@n7, Alpha=1, GravityOverride:Vector="0,0,-980", ExternalForce="0,0,0") -> (Pose) {
    angularDampingOverride=1
    angularSpringConstant=7000
    bAngularSpring=true
    bOverrideAngularDamping=true
    bUseGravityOverride=true
    gravityScale=0.35
    simulationSpace=RootRelative
    boundBone {boneName=dyn_R_sleeveHole}
    componentAppliedLinearAccClamp=<10000 10000 10000>
    componentLinearAccScale=<1 1 1>
    componentLinearVelScale=<1 1 1>
    gravityOverride {z=-980}
    physicsBodyDefinitions [{
      boundBone {boneName=dyn_R_sleeveHole}
      boxExtents=<30 40 40>
      constraintSetup {angularLimitsMax=<5 0 5>, angularLimitsMin=<-5 -5 -5>, angularTarget=<1 0 0>}
      localJointOffset=<-3 0 0>
     }]
   }
   n9 = Anim Dynamics - Bone: dyn_R_sleeveBig_1!(ComponentPose=@n8, Alpha=1, GravityOverride:Vector="0,0,-980", ExternalForce="0,0,0") -> (Pose) {
    angularDampingOverride=1
    angularSpringConstant=1000
    bAngularSpring=true
    bOverrideAngularDamping=true
    bUseGravityOverride=true
    gravityScale=0.35
    simulationSpace=RootRelative
    boundBone {boneName=dyn_R_sleeveBig_1}
    componentAppliedLinearAccClamp=<10000 10000 10000>
    componentLinearAccScale=<1 1 1>
    componentLinearVelScale=<1 1 1>
    gravityOverride {z=-980}
    physicsBodyDefinitions [{
      boundBone {boneName=dyn_R_sleeveBig_1}
      boxExtents=<30 40 40>
      constraintSetup {angularLimitsMax=<15 80 40>, angularLimitsMin=<-15 -80 -40>, angularTarget=<1 0 0>}
      localJointOffset=<-4 0 0>
     }]
   }
   n10 = Anim Dynamics - Bone: dyn_R_sleeveBig_2!(ComponentPose=@n9, Alpha=1, GravityOverride:Vector="0,0,-980", ExternalForce="0,0,0") -> (Pose) {
    angularDampingOverride=1
    angularSpringConstant=1000
    bAngularSpring=true
    bOverrideAngularDamping=true
    bUseGravityOverride=true
    gravityScale=0.35
    simulationSpace=RootRelative
    boundBone {boneName=dyn_R_sleeveBig_2}
    componentAppliedLinearAccClamp=<10000 10000 10000>
    componentLinearAccScale=<1 1 1>
    componentLinearVelScale=<1 1 1>
    gravityOverride {z=-980}
    physicsBodyDefinitions [{
      boundBone {boneName=dyn_R_sleeveBig_2}
      boxExtents=<30 40 40>
      constraintSetup {angularLimitsMax=<15 80 40>, angularLimitsMin=<-15 -80 -40>, angularTarget=<1 0 0>}
      localJointOffset=<-4 0 0>
     }]
   }
   n11 = Anim Dynamics - Bone: dyn_R_sleeveSmall_1!(ComponentPose=@n10, Alpha=1, GravityOverride:Vector="0,0,-980", ExternalForce="0,0,0") -> (Pose) {
    angularDampingOverride=1
    angularSpringConstant=1000
    bAngularSpring=true
    bOverrideAngularDamping=true
    bUseGravityOverride=true
    gravityScale=0.35
    simulationSpace=RootRelative
    boundBone {boneName=dyn_R_sleeveSmall_1}
    componentAppliedLinearAccClamp=<10000 10000 10000>
    componentLinearAccScale=<1 1 1>
    componentLinearVelScale=<1 1 1>
    gravityOverride {z=-980}
    physicsBodyDefinitions [{
      boundBone {boneName=dyn_R_sleeveSmall_1}
      boxExtents=<30 40 40>
      constraintSetup {angularLimitsMax=<15 80 40>, angularLimitsMin=<-15 -80 -40>, angularTarget=<1 0 0>}
      localJointOffset=<-4 0 0>
     }]
   }
   n12 = Anim Dynamics - Bone: dyn_R_sleeveSmall_2!(ComponentPose=@n11, Alpha=1, GravityOverride:Vector="0,0,-980", ExternalForce="0,0,0") -> (Pose) {
    angularDampingOverride=1
    angularSpringConstant=1000
    bAngularSpring=true
    bOverrideAngularDamping=true
    bUseGravityOverride=true
    gravityScale=0.35
    simulationSpace=RootRelative
    boundBone {boneName=dyn_R_sleeveSmall_2}
    componentAppliedLinearAccClamp=<10000 10000 10000>
    componentLinearAccScale=<1 1 1>
    componentLinearVelScale=<1 1 1>
    gravityOverride {z=-980}
    physicsBodyDefinitions [{
      boundBone {boneName=dyn_R_sleeveSmall_2}
      boxExtents=<30 40 40>
      constraintSetup {angularLimitsMax=<15 80 40>, angularLimitsMin=<-15 -80 -40>, angularTarget=<1 0 0>}
      localJointOffset=<-4 0 0>
     }]
   }
   Save cached pose 'Sleeve Dybn'(Pose=Component To Local(ComponentPose=@n12)) {cacheName=Sleeve Dybn}
   n13 = Anim Dynamics - Bone: dyn_L_haori_front_2!(
    ComponentPose=Local To Component(LocalPose=Use cached pose 'Sleeve Dybn'())
    Alpha=1
    GravityOverride:Vector="0,0,-980"
    ExternalForce="0,0,0"
   ) -> (Pose) {
    angularDampingOverride=1
    angularSpringConstant=12000
    bAngularSpring=true
    bOverrideAngularDamping=true
    bUseGravityOverride=true
    gravityScale=0.35
    simulationSpace=RootRelative
    boundBone {boneName=dyn_L_haori_front_2}
    componentAppliedLinearAccClamp=<10000 10000 10000>
    componentLinearAccScale=<1 1 1>
    componentLinearVelScale=<1 1 1>
    gravityOverride {z=-980}
    physicsBodyDefinitions [{
      boundBone {boneName=dyn_L_haori_front_2}
      boxExtents=<30 50 50>
      constraintSetup {angularLimitsMax=<8 0 10>, angularLimitsMin=<-8 -20 -10>, angularTarget=<1 0 0>}
      localJointOffset=<8 0 0>
     }]
   }
   n14 = Anim Dynamics - Bone: dyn_L_haori_front_3!(ComponentPose=@n13, Alpha=1, GravityOverride:Vector="0,0,-980", ExternalForce="0,0,0") -> (Pose) {
    angularDampingOverride=1
    angularSpringConstant=12000
    bAngularSpring=true
    bOverrideAngularDamping=true
    bUseGravityOverride=true
    gravityScale=0.35
    simulationSpace=RootRelative
    boundBone {boneName=dyn_L_haori_front_3}
    componentAppliedLinearAccClamp=<10000 10000 10000>
    componentLinearAccScale=<1 1 1>
    componentLinearVelScale=<1 1 1>
    gravityOverride {z=-980}
    physicsBodyDefinitions [{
      boundBone {boneName=dyn_L_haori_front_3}
      boxExtents=<30 50 50>
      constraintSetup {angularLimitsMax=<8 0 10>, angularLimitsMin=<-8 -20 -10>, angularTarget=<1 0 0>}
      localJointOffset=<8 0 0>
     }]
   }
   n15 = Anim Dynamics - Bone: dyn_L_haori_front_4!(ComponentPose=@n14, Alpha=1, GravityOverride:Vector="0,0,-980", ExternalForce="0,0,0") -> (Pose) {
    angularDampingOverride=1
    angularSpringConstant=12000
    bAngularSpring=true
    bOverrideAngularDamping=true
    bUseGravityOverride=true
    gravityScale=0.35
    simulationSpace=RootRelative
    boundBone {boneName=dyn_L_haori_front_4}
    componentAppliedLinearAccClamp=<10000 10000 10000>
    componentLinearAccScale=<1 1 1>
    componentLinearVelScale=<1 1 1>
    gravityOverride {z=-980}
    physicsBodyDefinitions [{
      boundBone {boneName=dyn_L_haori_front_4}
      boxExtents=<30 50 50>
      constraintSetup {angularLimitsMax=<8 0 10>, angularLimitsMin=<-8 -20 -10>, angularTarget=<1 0 0>}
      localJointOffset=<8 0 0>
     }]
   }
   n16 = Anim Dynamics - Bone: dyn_L_haori_side_2!(ComponentPose=@n15, Alpha=1, GravityOverride:Vector="0,0,-980", ExternalForce="0,0,0") -> (Pose) {
    angularDampingOverride=1
    angularSpringConstant=12000
    bAngularSpring=true
    bOverrideAngularDamping=true
    bUseGravityOverride=true
    gravityScale=0.35
    simulationSpace=RootRelative
    boundBone {boneName=dyn_L_haori_side_2}
    componentAppliedLinearAccClamp=<10000 10000 10000>
    componentLinearAccScale=<1 1 1>
    componentLinearVelScale=<1 1 1>
    gravityOverride {z=-980}
    physicsBodyDefinitions [{
      boundBone {boneName=dyn_L_haori_side_2}
      boxExtents=<30 50 50>
      constraintSetup {angularLimitsMax=<7 0 10>, angularLimitsMin=<-7 -20 -10>, angularTarget=<1 0 0>}
      localJointOffset=<8 0 0>
     }]
   }
   n17 = Anim Dynamics - Bone: dyn_L_haori_side_3!(ComponentPose=@n16, Alpha=1, GravityOverride:Vector="0,0,-980", ExternalForce="0,0,0") -> (Pose) {
    angularDampingOverride=1
    angularSpringConstant=12000
    bAngularSpring=true
    bOverrideAngularDamping=true
    bUseGravityOverride=true
    gravityScale=0.35
    simulationSpace=RootRelative
    boundBone {boneName=dyn_L_haori_side_3}
    componentAppliedLinearAccClamp=<10000 10000 10000>
    componentLinearAccScale=<1 1 1>
    componentLinearVelScale=<1 1 1>
    gravityOverride {z=-980}
    physicsBodyDefinitions [{
      boundBone {boneName=dyn_L_haori_side_3}
      boxExtents=<30 50 50>
      constraintSetup {angularLimitsMax=<7 0 10>, angularLimitsMin=<-7 -20 -10>, angularTarget=<1 0 0>}
      localJointOffset=<8 0 0>
     }]
   }
   n18 = Anim Dynamics - Bone: dyn_L_haori_side_4!(ComponentPose=@n17, Alpha=1, GravityOverride:Vector="0,0,-980", ExternalForce="0,0,0") -> (Pose) {
    angularDampingOverride=1
    angularSpringConstant=12000
    bAngularSpring=true
    bOverrideAngularDamping=true
    bUseGravityOverride=true
    gravityScale=0.35
    simulationSpace=RootRelative
    boundBone {boneName=dyn_L_haori_side_4}
    componentAppliedLinearAccClamp=<10000 10000 10000>
    componentLinearAccScale=<1 1 1>
    componentLinearVelScale=<1 1 1>
    gravityOverride {z=-980}
    physicsBodyDefinitions [{
      boundBone {boneName=dyn_L_haori_side_4}
      boxExtents=<30 50 50>
      constraintSetup {angularLimitsMax=<7 0 10>, angularLimitsMin=<-7 -20 -10>, angularTarget=<1 0 0>}
      localJointOffset=<8 0 0>
     }]
   }
   n19 = Anim Dynamics - Bone: dyn_C_haori_back_2!(ComponentPose=@n18, Alpha=1, GravityOverride:Vector="0,0,-980", ExternalForce="0,0,0") -> (Pose) {
    angularDampingOverride=1
    angularSpringConstant=12000
    bAngularSpring=true
    bOverrideAngularDamping=true
    bUseGravityOverride=true
    gravityScale=0.35
    simulationSpace=RootRelative
    boundBone {boneName=dyn_C_haori_back_2}
    componentAppliedLinearAccClamp=<10000 10000 10000>
    componentLinearAccScale=<1 1 1>
    componentLinearVelScale=<1 1 1>
    gravityOverride {z=-980}
    physicsBodyDefinitions [{
      boundBone {boneName=dyn_C_haori_back_2}
      boxExtents=<30 50 50>
      constraintSetup {angularLimitsMax=<8 0 10>, angularLimitsMin=<-8 -20 -10>, angularTarget=<1 0 0>}
      localJointOffset=<8 0 0>
     }]
   }
   n20 = Anim Dynamics - Bone: dyn_C_haori_back_3!(ComponentPose=@n19, Alpha=1, GravityOverride:Vector="0,0,-980", ExternalForce="0,0,0") -> (Pose) {
    angularDampingOverride=1
    angularSpringConstant=12000
    bAngularSpring=true
    bOverrideAngularDamping=true
    bUseGravityOverride=true
    gravityScale=0.35
    simulationSpace=RootRelative
    boundBone {boneName=dyn_C_haori_back_3}
    componentAppliedLinearAccClamp=<10000 10000 10000>
    componentLinearAccScale=<1 1 1>
    componentLinearVelScale=<1 1 1>
    gravityOverride {z=-980}
    physicsBodyDefinitions [{
      boundBone {boneName=dyn_C_haori_back_3}
      boxExtents=<30 50 50>
      constraintSetup {angularLimitsMax=<8 0 10>, angularLimitsMin=<-8 -20 -10>, angularTarget=<1 0 0>}
      localJointOffset=<8 0 0>
     }]
   }
   n21 = Anim Dynamics - Bone: dyn_C_haori_back_4!(ComponentPose=@n20, Alpha=1, GravityOverride:Vector="0,0,-980", ExternalForce="0,0,0") -> (Pose) {
    angularDampingOverride=1
    angularSpringConstant=12000
    bAngularSpring=true
    bOverrideAngularDamping=true
    bUseGravityOverride=true
    gravityScale=0.35
    simulationSpace=RootRelative
    boundBone {boneName=dyn_C_haori_back_4}
    componentAppliedLinearAccClamp=<10000 10000 10000>
    componentLinearAccScale=<1 1 1>
    componentLinearVelScale=<1 1 1>
    gravityOverride {z=-980}
    physicsBodyDefinitions [{
      boundBone {boneName=dyn_C_haori_back_4}
      boxExtents=<30 50 50>
      constraintSetup {angularLimitsMax=<8 0 10>, angularLimitsMin=<-8 -20 -10>, angularTarget=<1 0 0>}
      localJointOffset=<8 0 0>
     }]
   }
   n22 = Anim Dynamics - Bone: dyn_R_haori_front_2!(ComponentPose=@n21, Alpha=1, GravityOverride:Vector="0,0,-980", ExternalForce="0,0,0") -> (Pose) {
    angularDampingOverride=1
    angularSpringConstant=12000
    bAngularSpring=true
    bOverrideAngularDamping=true
    bUseGravityOverride=true
    gravityScale=0.35
    simulationSpace=RootRelative
    boundBone {boneName=dyn_R_haori_front_2}
    componentAppliedLinearAccClamp=<10000 10000 10000>
    componentLinearAccScale=<1 1 1>
    componentLinearVelScale=<1 1 1>
    gravityOverride {z=-980}
    physicsBodyDefinitions [{
      boundBone {boneName=dyn_R_haori_front_2}
      boxExtents=<30 50 50>
      constraintSetup {angularLimitsMax=<8 0 10>, angularLimitsMin=<-8 -20 -10>, angularTarget=<1 0 0>}
      localJointOffset=<8 0 0>
     }]
   }
   n23 = Anim Dynamics - Bone: dyn_R_haori_front_3!(ComponentPose=@n22, Alpha=1, GravityOverride:Vector="0,0,-980", ExternalForce="0,0,0") -> (Pose) {
    angularDampingOverride=1
    angularSpringConstant=12000
    bAngularSpring=true
    bOverrideAngularDamping=true
    bUseGravityOverride=true
    gravityScale=0.35
    simulationSpace=RootRelative
    boundBone {boneName=dyn_R_haori_front_3}
    componentAppliedLinearAccClamp=<10000 10000 10000>
    componentLinearAccScale=<1 1 1>
    componentLinearVelScale=<1 1 1>
    gravityOverride {z=-980}
    physicsBodyDefinitions [{
      boundBone {boneName=dyn_R_haori_front_3}
      boxExtents=<30 50 50>
      constraintSetup {angularLimitsMax=<8 0 10>, angularLimitsMin=<-8 -20 -10>, angularTarget=<1 0 0>}
      localJointOffset=<8 0 0>
     }]
   }
   n24 = Anim Dynamics - Bone: dyn_R_haori_front_4!(ComponentPose=@n23, Alpha=1, GravityOverride:Vector="0,0,-980", ExternalForce="0,0,0") -> (Pose) {
    angularDampingOverride=1
    angularSpringConstant=12000
    bAngularSpring=true
    bOverrideAngularDamping=true
    bUseGravityOverride=true
    gravityScale=0.35
    simulationSpace=RootRelative
    boundBone {boneName=dyn_R_haori_front_4}
    componentAppliedLinearAccClamp=<10000 10000 10000>
    componentLinearAccScale=<1 1 1>
    componentLinearVelScale=<1 1 1>
    gravityOverride {z=-980}
    physicsBodyDefinitions [{
      boundBone {boneName=dyn_R_haori_front_4}
      boxExtents=<30 50 50>
      constraintSetup {angularLimitsMax=<8 0 10>, angularLimitsMin=<-8 -20 -10>, angularTarget=<1 0 0>}
      localJointOffset=<8 0 0>
     }]
   }
   n25 = Anim Dynamics - Bone: dyn_R_haori_side_2!(ComponentPose=@n24, Alpha=1, GravityOverride:Vector="0,0,-980", ExternalForce="0,0,0") -> (Pose) {
    angularDampingOverride=1
    angularSpringConstant=12000
    bAngularSpring=true
    bOverrideAngularDamping=true
    bUseGravityOverride=true
    gravityScale=0.35
    simulationSpace=RootRelative
    boundBone {boneName=dyn_R_haori_side_2}
    componentAppliedLinearAccClamp=<10000 10000 10000>
    componentLinearAccScale=<1 1 1>
    componentLinearVelScale=<1 1 1>
    gravityOverride {z=-980}
    physicsBodyDefinitions [{
      boundBone {boneName=dyn_R_haori_side_2}
      boxExtents=<30 50 50>
      constraintSetup {angularLimitsMax=<8 0 10>, angularLimitsMin=<-8 -20 -10>, angularTarget=<1 0 0>}
      localJointOffset=<8 0 0>
     }]
   }
   n26 = Anim Dynamics - Bone: dyn_R_haori_side_3!(ComponentPose=@n25, Alpha=1, GravityOverride:Vector="0,0,-980", ExternalForce="0,0,0") -> (Pose) {
    angularDampingOverride=1
    angularSpringConstant=12000
    bAngularSpring=true
    bOverrideAngularDamping=true
    bUseGravityOverride=true
    gravityScale=0.35
    simulationSpace=RootRelative
    boundBone {boneName=dyn_R_haori_side_3}
    componentAppliedLinearAccClamp=<10000 10000 10000>
    componentLinearAccScale=<1 1 1>
    componentLinearVelScale=<1 1 1>
    gravityOverride {z=-980}
    physicsBodyDefinitions [{
      boundBone {boneName=dyn_R_haori_side_3}
      boxExtents=<30 50 50>
      constraintSetup {angularLimitsMax=<8 0 10>, angularLimitsMin=<-8 -20 -10>, angularTarget=<1 0 0>}
      localJointOffset=<8 0 0>
     }]
   }
   n27 = Anim Dynamics - Bone: dyn_R_haori_side_4!(ComponentPose=@n26, Alpha=1, GravityOverride:Vector="0,0,-980", ExternalForce="0,0,0") -> (Pose) {
    angularDampingOverride=1
    angularSpringConstant=12000
    bAngularSpring=true
    bOverrideAngularDamping=true
    bUseGravityOverride=true
    gravityScale=0.35
    simulationSpace=RootRelative
    boundBone {boneName=dyn_R_haori_side_4}
    componentAppliedLinearAccClamp=<10000 10000 10000>
    componentLinearAccScale=<1 1 1>
    componentLinearVelScale=<1 1 1>
    gravityOverride {z=-980}
    physicsBodyDefinitions [{
      boundBone {boneName=dyn_R_haori_side_4}
      boxExtents=<30 50 50>
      constraintSetup {angularLimitsMax=<8 0 10>, angularLimitsMin=<-8 -20 -10>, angularTarget=<1 0 0>}
      localJointOffset=<8 0 0>
     }]
   }
   Save cached pose 'Haori Dyn'(Pose=Component To Local(ComponentPose=@n27)) {cacheName=Haori Dyn}
   n28 = "Transform (Modify) Bone - Bone: attach_backpack"(ComponentPose=Local To Component(LocalPose=Use cached pose 'ControlRig'()), Alpha=1, Translation="0,-4,0", Rotation="0,0,0", Scale="1,1,1") {
    rotationMode=BMM_Additive
    rotationSpace=BCS_BoneSpace
    translationMode=BMM_Additive
    translationSpace=BCS_BoneSpace
    boneToModify {boneName=attach_backpack}
    translation {y=-4}
   }
   n29 = "Transform (Modify) Bone - Bone: attach_cape"(ComponentPose=@n28, Alpha=1, Translation="-0.629,-8,-1.44", Rotation="0,0,0", Scale="1,1,1") {translationMode=BMM_Additive, translationSpace=BCS_BoneSpace, boneToModify {boneName=attach_cape}, translation=<-0.629 -8 -1.44>}
   Save cached pose 'Predynamics'(Pose=Component To Local(ComponentPose=@n29)) {cacheName=Predynamics}
  ]}
 EventGraph {nodes [
   Event BlueprintUpdateAnimation()
   > n1 = Sequence!(execute) -> (then_0, then_1, then_2:Exec)
   > n2 = "Gravity Override State Function (Parameter Asset) (Nativized)"(
    CharacterPartAnimInstance=Self-Reference()
    GameTimeInSeconds=GetGameTimeInSeconds()
    Parameters=/BRCosmetics/Characters/Player/Female/Medium/Bodies/F_MED_BrightLarva/Meshes/F_MED_BrightLarva_Parameters
   )
   > Set GravityOverride(GravityOverride=@n2)
   n3 = GetAnimInstance(self=GetOwnerSkeletalMeshComponent())
   > Is Valid(InputObject=@n3)
   > Set IsEasterBloc(IsEasterBloc:bool=Montage_IsPlaying(self=@n3, Montage=/BRCosmetics/Animation/Game/MainPlayer/Emotes/Eastern_Bloc/Emote_EasternBloc_CMF_M)) -> (Output_Get:bool)
   > Set IsYouBoreme(IsYouBoreme:bool=Montage_IsPlaying(self=@n3, Montage=/BRCosmetics/Animation/Game/MainPlayer/Emotes/YouBoreMe/Emote_You_Bore_Me_CMF_M)) -> (Output_Get:bool)
   n4 = PlayerMovement(execute=@n1.then_1, AnimInstance=Self-Reference(), XDotProductCoefficient=1, YDotProductCoefficient=1, ZDotProductCoefficient=1)
   n5 = BreakVector(InVec=MakeVector(X=@n4, Y=@n4.YMovement))
   > Set strafeLeft float(strafeLeft float:double="Absolute (Float)"(A="Clamp (Float)"(Value=@n5, Min=-1, Max=0))) -> (Output_Get:double)
   > Set strafeRight float(strafeRight float:double="Clamp (Float)"(Value=@n5, Min=0, Max=1)) -> (Output_Get:double)
   > Set moveForward(moveForward:double="Clamp (Float)"(Value=@n5.Y, Min=0, Max=1)) -> (Output_Get:double)
   > Set moveBackward(moveBackward:double="Absolute (Float)"(A="Clamp (Float)"(Value=@n5.Y, Min=-1, Max=0))) -> (Output_Get:double)
   > n6 = Branch(Condition=OR Boolean!(A=Get isJumping(), B=Get bIsFalling(), C:bool=Get bIsKneeSliding(), D:bool=Get bIsBaseballSliding()) -> (ReturnValue))
   > Set AlphaDyn(AlphaDyn:double=0.3) -> (Output_Get:double)
   > Set AlphaDynHaori(AlphaDynHaori:double=0.5) -> (Output_Get:double)
   Set AlphaDyn(execute=@n6.else, AlphaDyn=0.5)
   > Set AlphaDynHaori(AlphaDynHaori=1)
   n7 = Branch(execute=@n1.then_2, Condition="integer > integer"(A=Get CurrentLOD(), B=1) {is=title K2Node_CallFunction})
   > SetMorphTarget!(execute, self:Anim Instance Object Reference, MorphTargetName:Name=Cloth_LODs_morphs, Value:float=1) -> (then)
   > SetMorphTarget(MorphTargetName=Body_LODs_morphs, Value=1)
   SetMorphTarget(execute=@n7.else, MorphTargetName=Cloth_LODs_morphs, Value=0)
   > SetMorphTarget(MorphTargetName=Body_LODs_morphs, Value=0)
   Event OnBeginJumping()
   > Set isJumping(isJumping:bool=true) -> (Output_Get:bool)
   Event OnBeginPickaxeSwing() {is=title K2Node_Event}
   > Set IsPASwing(IsPASwing:bool=true) -> (Output_Get:bool)
   Event OnCharacterCustomizationLoaded()
   > n8 = Cast To PlayerPawn_Athena(Object=TryGetPawnOwner())
   > "Set Face Mesh "("Face Mesh ":Skeletal Mesh Component Object Reference=GetSkeletalMeshForPartType(self=@n8, PartType=Head)) -> (Output_Get:Skeletal Mesh Component Object Reference)
   Event OnEndPickaxeSwing() {is=title K2Node_Event}
   > Set IsPASwing(IsPASwing=false)
   Event OnEnterGrindRail() {is=title K2Node_Event}
   > Set isOnGrindRail float(isOnGrindRail float:double=1) -> (Output_Get:double)
   Event OnExitGrindRail() {is=title K2Node_Event}
   > Set isOnGrindRail float(isOnGrindRail float=0)
   Event OnEndFalling()
   > Set isJumping(isJumping=false)
  ]}
]
variables [
 AlphaDyn {type=number, default=0.5}
 AlphaDynHaori {type=number, default=1}
 "Face Mesh " {type=SkeletalMeshComponent}
 GravityOverride {type=Vector}
 IsEasterBloc {type=boolean}
 isJumping {type=boolean}
 isOnGrindRail float {type=number}
 IsPASwing {type=boolean}
 IsYouBoreme {type=boolean}
 moveBackward {type=number}
 moveForward {type=number}
 strafeLeft float {type=number}
 strafeRight float {type=number}
]
