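# Physics asset for the F_MED_AtticDrill_Fin female character body, providing capsule bodies for the torso, legs, feet, hapi garment and skirt bones.
# Five hapi chains hang from base bodies rotation-locked to spine_04, with angular limits down each chain and twist limits differing between side and back.
# Free-rotation cross constraints tie the center back hapi chain to its neighbors at two levels.
# Skirt chains, calves and feet have bodies but no constraints here.
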PhysicsAsset /BRCosmetics/Characters/Player/Female/Medium/Bodies/F_MED_AtticDrill_Fin/Meshes/F_MED_AtticDrill_Fin_Physics
bodies {
 base_hapi_back_00 [{shapeType=Capsule, length=0.505, radius=0.505, shapeName=base_hapi_back_00_capsule, rotation=<rot 0 -90 0>}]
 base_hapi_back_lf_00 [{shapeType=Capsule, length=0.505, radius=0.505, shapeName=base_hapi_back_lf_00_capsule, rotation=<rot 0 -90 0>}]
 base_hapi_back_rt_00 [{shapeType=Capsule, length=0.505, radius=0.505, shapeName=base_hapi_back_rt_00_capsule, rotation=<rot 0 -90 0>}]
 base_hapi_left_00 [{shapeType=Capsule, length=0.505, radius=0.505, shapeName=base_hapi_left_00_capsule, rotation=<rot 0 -90 0>}]
 base_hapi_right_00 [{shapeType=Capsule, length=0.505, radius=0.505, shapeName=base_hapi_right_00_capsule, rotation=<rot 0 -90 0>}]
 base_skirt_back_00 [{shapeType=Capsule, length=0.505, radius=5, shapeName=base_skirt_back_00_capsule}]
 base_skirt_back_00_lf [{shapeType=Capsule, length=0.505, radius=5, shapeName=base_skirt_back_00_lf_capsule}]
 base_skirt_back_00_rt [{shapeType=Capsule, length=0.505, radius=5, shapeName=base_skirt_back_00_rt_capsule}]
 base_skirt_side_00_lf [{shapeType=Capsule, length=0.505, radius=5, shapeName=base_skirt_side_00_lf_capsule}]
 base_skirt_side_00_rt [{shapeType=Capsule, length=0.505, radius=5, shapeName=base_skirt_side_00_rt_capsule}]
 calf_l [{shapeType=Capsule, length=33.2654, radius=7.7667, shapeName=calf_l_capsule, center=<-26.5038 -2.2792 -0.7458>, rotation=<rot 90 0 0>}]
 calf_r [{shapeType=Capsule, length=33.2654, radius=7.7667, shapeName=calf_r_capsule, center=<26.5038 2.2792 0.7458>, rotation=<rot 90 180 0>}]
 dyn_hapi_back_00 [{shapeType=Capsule, length=20, radius=3.5, shapeName=dyn_hapi_back_00_capsule, center=<4 0 0>}]
 dyn_hapi_back_01 [{shapeType=Capsule, length=20, radius=3.5, shapeName=dyn_hapi_back_01_capsule, center=<4 -2 0>}]
 dyn_hapi_back_02 [{shapeType=Capsule, length=20, radius=3.5, shapeName=dyn_hapi_back_02_capsule, center=<4 0 0>}]
 dyn_hapi_back_03 [{shapeType=Capsule, length=20, radius=3.5, shapeName=dyn_hapi_back_03_capsule, center=<4 0 0>}]
 dyn_hapi_back_04 [{shapeType=Capsule, length=20, radius=3.5, shapeName=dyn_hapi_back_04_capsule, center=<4 0 0>}]
 dyn_hapi_back_lf_00 [{shapeType=Capsule, length=20, radius=3.5, shapeName=dyn_hapi_back_lf_00_capsule, center=<4 0 0>}]
 dyn_hapi_back_lf_01 [{shapeType=Capsule, length=20, radius=3.5, shapeName=dyn_hapi_back_lf_01_capsule, center=<4 0 0>}]
 dyn_hapi_back_lf_02 [{shapeType=Capsule, length=20, radius=3.5, shapeName=dyn_hapi_back_lf_02_capsule, center=<4 -2 0>}]
 dyn_hapi_back_lf_03 [{shapeType=Capsule, length=20, radius=3.5, shapeName=dyn_hapi_back_lf_03_capsule, center=<4 0 0>}]
 dyn_hapi_back_lf_04 [{shapeType=Capsule, length=20, radius=3.5, shapeName=dyn_hapi_back_lf_04_capsule, center=<4 0 0>}]
 dyn_hapi_back_rt_00 [{shapeType=Capsule, length=20, radius=3.5, shapeName=dyn_hapi_back_rt_00_capsule, center=<4 0 0>}]
 dyn_hapi_back_rt_01 [{shapeType=Capsule, length=19.5, radius=4, shapeName=dyn_hapi_back_rt_01_capsule, center=<4 0 0>}]
 dyn_hapi_back_rt_02 [{shapeType=Capsule, length=20, radius=3.5, shapeName=dyn_hapi_back_rt_02_capsule, center=<4 -0.3 0>}]
 dyn_hapi_back_rt_03 [{shapeType=Capsule, length=20, radius=3.5, shapeName=dyn_hapi_back_rt_03_capsule, center=<4 0 0>}]
 dyn_hapi_back_rt_04 [{shapeType=Capsule, length=20, radius=3.5, shapeName=dyn_hapi_back_rt_04_capsule, center=<4 0 0>}]
 dyn_hapi_left_00 [{shapeType=Capsule, length=20, radius=3.5, shapeName=dyn_hapi_left_00_capsule, center=<4 0 0>}]
 dyn_hapi_left_01 [{shapeType=Capsule, length=20.5, radius=3, shapeName=dyn_hapi_left_01_capsule, center=<4 0 0>}]
 dyn_hapi_left_02 [{shapeType=Capsule, length=20, radius=3.5, shapeName=dyn_hapi_left_02_capsule, center=<4 0 0>}]
 dyn_hapi_left_03 [{shapeType=Capsule, length=20, radius=3.5, shapeName=dyn_hapi_left_03_capsule, center=<4 0 0>}]
 dyn_hapi_left_04 [{shapeType=Capsule, length=20, radius=3.5, shapeName=dyn_hapi_left_04_capsule, center=<4 0 0>}]
 dyn_hapi_right_00 [{shapeType=Capsule, length=20, radius=3.5, shapeName=dyn_hapi_right_00_capsule, center=<4 0 0>}]
 dyn_hapi_right_01 [{shapeType=Capsule, length=20.5, radius=3, shapeName=dyn_hapi_right_01_capsule, center=<4 0 0>}]
 dyn_hapi_right_02 [{shapeType=Capsule, length=20, radius=3.5, shapeName=dyn_hapi_right_02_capsule, center=<4 0 0>}]
 dyn_hapi_right_03 [{shapeType=Capsule, length=20, radius=3.5, shapeName=dyn_hapi_right_03_capsule, center=<4 0.8272 0>}]
 dyn_hapi_right_04 [{shapeType=Capsule, length=20, radius=3.5, shapeName=dyn_hapi_right_04_capsule, center=<4 0 0>}]
 dyn_skirt_back_00 [{shapeType=Capsule, length=13.0581, radius=5, shapeName=dyn_skirt_back_00_capsule}]
 dyn_skirt_back_00_lf [{shapeType=Capsule, length=12.1969, radius=2.5, shapeName=dyn_skirt_back_00_lf_capsule}]
 dyn_skirt_back_00_rt [{shapeType=Capsule, length=5.0371, radius=3, shapeName=dyn_skirt_back_00_rt_capsule}]
 dyn_skirt_back_01 [{shapeType=Capsule, length=17.2883, radius=5, shapeName=dyn_skirt_back_01_capsule}]
 dyn_skirt_back_01_lf [{shapeType=Capsule, length=13.14, radius=5, shapeName=dyn_skirt_back_01_lf_capsule}]
 dyn_skirt_back_01_rt [{shapeType=Capsule, length=14.7102, radius=5, shapeName=dyn_skirt_back_01_rt_capsule}]
 dyn_skirt_back_02 [{shapeType=Capsule, length=10.8571, radius=5, shapeName=dyn_skirt_back_02_capsule}]
 dyn_skirt_back_02_lf [{shapeType=Capsule, length=9.292, radius=5, shapeName=dyn_skirt_back_02_lf_capsule}]
 dyn_skirt_back_02_rt [{shapeType=Capsule, length=7.7501, radius=5, shapeName=dyn_skirt_back_02_rt_capsule}]
 dyn_skirt_back_03 [{shapeType=Capsule, length=16.8112, radius=7.5, shapeName=dyn_skirt_back_03_capsule, center=<0 2.1936 0>, rotation=<rot -90 -0.0001 0>}]
 dyn_skirt_back_03_lf [{shapeType=Capsule, length=14.0988, radius=5, shapeName=dyn_skirt_back_03_lf_capsule}]
 dyn_skirt_back_03_rt [{shapeType=Capsule, length=15.2617, radius=5, shapeName=dyn_skirt_back_03_rt_capsule}]
 dyn_skirt_side_00_lf [{shapeType=Capsule, length=12.3628, radius=5, shapeName=dyn_skirt_side_00_lf_capsule}]
 dyn_skirt_side_00_rt [{shapeType=Capsule, length=12.2162, radius=5, shapeName=dyn_skirt_side_00_rt_capsule}]
 dyn_skirt_side_01_lf [{shapeType=Capsule, length=15.4983, radius=5, shapeName=dyn_skirt_side_01_lf_capsule, center=<0 2.1364 0>}]
 dyn_skirt_side_01_rt [{shapeType=Capsule, length=11.2272, radius=5, shapeName=dyn_skirt_side_01_rt_capsule, center=<1.8884 -0.9055 0>}]
 dyn_skirt_side_02_lf [{shapeType=Capsule, length=6.5644, radius=5, shapeName=dyn_skirt_side_02_lf_capsule}]
 dyn_skirt_side_02_rt [{shapeType=Capsule, length=7.2366, radius=5, shapeName=dyn_skirt_side_02_rt_capsule, center=<0 -2.6685 0>}]
 dyn_skirt_side_03_lf [{shapeType=Capsule, length=11.7742, radius=6.5, shapeName=dyn_skirt_side_03_lf_capsule}]
 dyn_skirt_side_03_rt [{shapeType=Capsule, length=15.2135, radius=4.5, shapeName=dyn_skirt_side_03_rt_capsule, center=<-2.7628 -2.3079 0>}]
 foot_l [{
   shapeType=Capsule
   length=12.2347
   radius=7.5997
   shapeName=foot_l_capsule
   center=<-3.6354 6.7834 -0.1966>
   rotation=<rot 0 20.2216 -88.9367>
  }]
 foot_r [{shapeType=Capsule, length=12.2358, radius=7.599, shapeName=foot_r_capsule, center=<3.6342 -6.7888 0.2088>, rotation=<rot 0 20.1962 -89.0856>}]
 pelvis [{shapeType=Capsule, length=8.8299, radius=10.933, center=<-1.909 -2.2429 0.1675>, rotation=<rot 2.7058 -90 0.4899>}]
 spine_02 [{shapeType=Capsule, length=2.3472, radius=11.5038, center=<2.1695 -0.4686 -0.3124>, rotation=<rot 0 -168.7259 -171.2994>}]
 spine_04 [{shapeType=Capsule, length=18.7016, radius=15.3486, center=<1.5369 1.1914 0.3064>, rotation=<rot 1.483 -90 -1.2837>}]
 thigh_l [{shapeType=Capsule, length=33.6025, radius=6.8135, center=<-26.0516 -0.8161 1.0232>, rotation=<rot 0.5204 93.2718 95.959>}]
 thigh_r [{shapeType=Capsule, length=33.6025, radius=6.8135, center=<26.0387 1.0409 -1.0212>, rotation=<rot 0.5204 93.2718 95.959>}]
}
constraints [
 {bone1Name=spine_02, bone2Name=pelvis, swing1LimitDegrees=45, swing2LimitDegrees=45, twistLimitDegrees=45}
 {bone1Name=spine_04, bone2Name=spine_02, swing1LimitDegrees=45, swing2LimitDegrees=45, twistLimitDegrees=45}
 {bone1Name=thigh_l, bone2Name=pelvis, swing1LimitDegrees=45, swing2LimitDegrees=45, twistLimitDegrees=45}
 {bone1Name=thigh_r, bone2Name=pelvis, swing1LimitDegrees=45, swing2LimitDegrees=45, twistLimitDegrees=45}
 {bone1Name=dyn_hapi_right_04, bone2Name=dyn_hapi_right_03, swing1LimitDegrees=30, swing2LimitDegrees=15, twistLimitDegrees=20}
 {
  bone1Name=base_hapi_back_00
  bone2Name=spine_04
  swing1LimitDegrees=45
  swing1Motion=Locked
  swing2LimitDegrees=45
  swing2Motion=Locked
  twistLimitDegrees=45
  twistMotion=Locked
 }
 {bone1Name=dyn_hapi_back_00, bone2Name=base_hapi_back_00, swing1LimitDegrees=30, swing2LimitDegrees=5, twistLimitDegrees=0}
 {bone1Name=dyn_hapi_back_01, bone2Name=dyn_hapi_back_00, swing1LimitDegrees=30, swing2LimitDegrees=5, twistLimitDegrees=0}
 {bone1Name=dyn_hapi_back_02, bone2Name=dyn_hapi_back_01, swing1LimitDegrees=30, swing2LimitDegrees=5, twistLimitDegrees=0}
 {bone1Name=dyn_hapi_back_03, bone2Name=dyn_hapi_back_02, swing1LimitDegrees=30, swing2LimitDegrees=5, twistLimitDegrees=0}
 {bone1Name=dyn_hapi_back_04, bone2Name=dyn_hapi_back_03, swing1LimitDegrees=30, swing2LimitDegrees=5, twistLimitDegrees=0}
 {
  bone1Name=base_hapi_back_lf_00
  bone2Name=spine_04
  swing1LimitDegrees=45
  swing1Motion=Locked
  swing2LimitDegrees=45
  swing2Motion=Locked
  twistLimitDegrees=45
  twistMotion=Locked
 }
 {bone1Name=dyn_hapi_back_lf_00, bone2Name=base_hapi_back_lf_00, swing1LimitDegrees=30, swing2LimitDegrees=5, twistLimitDegrees=0}
 {bone1Name=dyn_hapi_back_lf_01, bone2Name=dyn_hapi_back_lf_00, swing1LimitDegrees=30, swing2LimitDegrees=5, twistLimitDegrees=0}
 {bone1Name=dyn_hapi_back_lf_02, bone2Name=dyn_hapi_back_lf_01, swing1LimitDegrees=30, swing2LimitDegrees=5, twistLimitDegrees=0}
 {bone1Name=dyn_hapi_back_lf_03, bone2Name=dyn_hapi_back_lf_02, swing1LimitDegrees=30, swing2LimitDegrees=5, twistLimitDegrees=0}
 {bone1Name=dyn_hapi_back_lf_04, bone2Name=dyn_hapi_back_lf_03, swing1LimitDegrees=30, swing2LimitDegrees=5, twistLimitDegrees=0}
 {
  bone1Name=base_hapi_back_rt_00
  bone2Name=spine_04
  swing1LimitDegrees=45
  swing1Motion=Locked
  swing2LimitDegrees=45
  swing2Motion=Locked
  twistLimitDegrees=45
  twistMotion=Locked
 }
 {bone1Name=dyn_hapi_back_rt_00, bone2Name=base_hapi_back_rt_00, swing1LimitDegrees=30, swing2LimitDegrees=5, twistLimitDegrees=0}
 {bone1Name=dyn_hapi_back_rt_01, bone2Name=dyn_hapi_back_rt_00, swing1LimitDegrees=30, swing2LimitDegrees=5, twistLimitDegrees=0}
 {bone1Name=dyn_hapi_back_rt_02, bone2Name=dyn_hapi_back_rt_01, swing1LimitDegrees=30, swing2LimitDegrees=5, twistLimitDegrees=0}
 {bone1Name=dyn_hapi_back_rt_03, bone2Name=dyn_hapi_back_rt_02, swing1LimitDegrees=30, swing2LimitDegrees=5, twistLimitDegrees=0}
 {bone1Name=dyn_hapi_back_rt_04, bone2Name=dyn_hapi_back_rt_03, swing1LimitDegrees=30, swing2LimitDegrees=5, twistLimitDegrees=0}
 {
  bone1Name=base_hapi_left_00
  bone2Name=spine_04
  swing1LimitDegrees=45
  swing1Motion=Locked
  swing2LimitDegrees=45
  swing2Motion=Locked
  twistLimitDegrees=45
  twistMotion=Locked
 }
 {bone1Name=dyn_hapi_left_00, bone2Name=base_hapi_left_00, swing1LimitDegrees=30, swing2LimitDegrees=5, twistLimitDegrees=20}
 {bone1Name=dyn_hapi_left_01, bone2Name=dyn_hapi_left_00, swing1LimitDegrees=30, swing2LimitDegrees=5, twistLimitDegrees=20}
 {bone1Name=dyn_hapi_left_02, bone2Name=dyn_hapi_left_01, swing1LimitDegrees=30, swing2LimitDegrees=5, twistLimitDegrees=20}
 {bone1Name=dyn_hapi_left_03, bone2Name=dyn_hapi_left_02, swing1LimitDegrees=30, swing2LimitDegrees=5, twistLimitDegrees=20}
 {bone1Name=dyn_hapi_left_04, bone2Name=dyn_hapi_left_03, swing1LimitDegrees=30, swing2LimitDegrees=15, twistLimitDegrees=20}
 {
  bone1Name=base_hapi_right_00
  bone2Name=spine_04
  swing1LimitDegrees=45
  swing1Motion=Locked
  swing2LimitDegrees=45
  swing2Motion=Locked
  twistLimitDegrees=45
  twistMotion=Locked
 }
 {bone1Name=dyn_hapi_right_00, bone2Name=base_hapi_right_00, swing1LimitDegrees=30, swing2LimitDegrees=5, twistLimitDegrees=20}
 {bone1Name=dyn_hapi_right_01, bone2Name=dyn_hapi_right_00, swing1LimitDegrees=30, swing2LimitDegrees=5, twistLimitDegrees=20}
 {bone1Name=dyn_hapi_right_02, bone2Name=dyn_hapi_right_01, swing1LimitDegrees=30, swing2LimitDegrees=5, twistLimitDegrees=20}
 {bone1Name=dyn_hapi_right_03, bone2Name=dyn_hapi_right_02, swing1LimitDegrees=30, swing2LimitDegrees=5, twistLimitDegrees=20}
 {
  bone1Name=dyn_hapi_back_01
  bone2Name=dyn_hapi_back_lf_01
  swing1LimitDegrees=45
  swing1Motion=Free
  swing2LimitDegrees=45
  swing2Motion=Free
  twistLimitDegrees=45
  twistMotion=Free
 }
 {
  bone1Name=dyn_hapi_back_01
  bone2Name=dyn_hapi_back_rt_01
  swing1LimitDegrees=45
  swing1Motion=Free
  swing2LimitDegrees=45
  swing2Motion=Free
  twistLimitDegrees=45
  twistMotion=Free
 }
 {
  bone1Name=dyn_hapi_back_03
  bone2Name=dyn_hapi_back_lf_03
  swing1LimitDegrees=45
  swing1Motion=Free
  swing2LimitDegrees=45
  swing2Motion=Free
  twistLimitDegrees=45
  twistMotion=Free
 }
 {
  bone1Name=dyn_hapi_back_03
  bone2Name=dyn_hapi_back_rt_03
  swing1LimitDegrees=45
  swing1Motion=Free
  swing2LimitDegrees=45
  swing2Motion=Free
  twistLimitDegrees=45
  twistMotion=Free
 }
]
