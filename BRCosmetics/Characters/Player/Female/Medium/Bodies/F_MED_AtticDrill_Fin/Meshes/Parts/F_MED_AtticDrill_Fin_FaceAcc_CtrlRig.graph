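# Control Rig for the F_MED_AtticDrill_Fin face accessory that procedurally adjusts left, center and right hair roots from body pose.
# Spherical pose readers on the head and clavicles produce weights that shift an aim target parented under spine_05.
# Local-space aim constraints point the hair at that target, then weighted corrective transforms adjust base and offset bones.
# Two final head-driven target shifts run after the aim constraints.
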ControlRigBlueprint /BRCosmetics/Characters/Player/Female/Medium/Bodies/F_MED_AtticDrill_Fin/Meshes/Parts/F_MED_AtticDrill_Fin_FaceAcc_CtrlRig
graphs [
 CollapseNode_ContainedGraph {nodes [
   n1 = RigVMFunctionEntryNode()
   n2 = &RigVMFunction_Sequence(ExecuteContext=@n1.ExecuteContext)
   RigVMFunction_Sequence_2 = RigVMAggregateNode(ExecuteContext=@n2.A)
   n3 = &RigUnit_SphericalPoseReader(
    ActiveRegionScaleFactors=(PositiveWidth=1.35,NegativeWidth=1.35,PositiveHeight=0.45,NegativeHeight=1.9)
    ActiveRegionSize=0.1
    Debug=(bDrawDebug=false,bDraw2D=False,bDrawLocalAxes=False,DebugScale=25,DebugSegments=20,DebugThickness=0.25)
    DriverAxis=(X=1,Y=1,Z=0)
    DriverItem=(Type=Bone,Name="head")
    FalloffRegionScaleFactors=(PositiveWidth=1,NegativeWidth=1,PositiveHeight=1.0266,NegativeHeight=1)
    FalloffSize=0.235
    FlipHeightScaling=False
    FlipWidthScaling=False
    OptionalParentItem=(Type=Bone,Name="None")
    RotationOffset=(X=0,Y=-90,Z=0)
    ExecutePin=@RigVMFunction_Sequence_2.A
   )
   > n4 = &RigUnit_SphericalPoseReader(
    ActiveRegionScaleFactors=(PositiveWidth=1.35,NegativeWidth=1.35,PositiveHeight=1,NegativeHeight=1.9)
    ActiveRegionSize=0.1968
    Debug=(bDrawDebug=false,bDraw2D=False,bDrawLocalAxes=False,DebugScale=25,DebugSegments=20,DebugThickness=0.25)
    DriverAxis=(X=1,Y=1,Z=0)
    DriverItem=(Type=Bone,Name="head")
    FalloffRegionScaleFactors=(PositiveWidth=1,NegativeWidth=1,PositiveHeight=1.0266,NegativeHeight=1)
    FalloffSize=0.235
    FlipHeightScaling=False
    FlipWidthScaling=False
    OptionalParentItem=(Type=Bone,Name="None")
    RotationOffset=(X=-37,Y=-90,Z=0)
   )
   > n5 = &RigUnit_SphericalPoseReader(
    ActiveRegionScaleFactors=(PositiveWidth=0.21,NegativeWidth=1,PositiveHeight=1,NegativeHeight=1)
    ActiveRegionSize=0.5
    Debug=(bDrawDebug=false,bDraw2D=False,bDrawLocalAxes=False,DebugScale=25,DebugSegments=20,DebugThickness=0.25)
    DriverAxis=(X=1,Y=0,Z=0)
    DriverItem=(Type=Bone,Name="head")
    FalloffRegionScaleFactors=(PositiveWidth=1,NegativeWidth=1,PositiveHeight=1,NegativeHeight=1)
    FalloffSize=0.5345
    FlipHeightScaling=False
    FlipWidthScaling=False
    OptionalParentItem=(Type=Bone,Name="None")
    RotationOffset=(X=90,Y=0,Z=6.2359)
   )
   > RigVMVariableNode(Value=@n5.OutputParam) {variable=Head Down}
   > n6 = &RigUnit_SphericalPoseReader(
    ActiveRegionScaleFactors=(PositiveWidth=1,NegativeWidth=1,PositiveHeight=1,NegativeHeight=1)
    ActiveRegionSize=0.15
    Debug=(bDrawDebug=false,bDraw2D=false,DebugScale=25,DebugSegments=20,DebugThickness=0.25)
    DriverAxis=(X=0,Y=1,Z=0)
    DriverItem=(Type=Bone,Name="head")
    FalloffRegionScaleFactors=(PositiveWidth=1,NegativeWidth=1,PositiveHeight=1,NegativeHeight=1)
    FalloffSize=0.32
    OptionalParentItem=(Type=Bone)
    RotationOffset=(X=150,Y=0,Z=0)
   )
   > RigVMVariableNode(Value=@n6.OutputParam) {variable=Head Left}
   > n7 = &RigUnit_SphericalPoseReader(
    ActiveRegionScaleFactors=(PositiveWidth=1,NegativeWidth=1,PositiveHeight=1,NegativeHeight=1)
    ActiveRegionSize=0.15
    Debug=(bDrawDebug=false,bDraw2D=false,DebugScale=25,DebugSegments=20,DebugThickness=0.25)
    DriverAxis=(X=0,Y=1,Z=0)
    DriverItem=(Type=Bone,Name="head")
    FalloffRegionScaleFactors=(PositiveWidth=1,NegativeWidth=1,PositiveHeight=1,NegativeHeight=1)
    FalloffSize=0.32
    FlipHeightScaling=true
    FlipWidthScaling=false
    OptionalParentItem=(Type=Bone)
    RotationOffset=(X=30,Y=0,Z=0)
   )
   > RigVMVariableNode(Value=@n7.OutputParam) {variable=Head Right}
   > n8 = &RigUnit_SphericalPoseReader(
    ActiveRegionScaleFactors=(PositiveWidth=3,NegativeWidth=1,PositiveHeight=3,NegativeHeight=1)
    ActiveRegionSize=0.1
    Debug=(bDrawDebug=false,bDraw2D=False,bDrawLocalAxes=False,DebugScale=25,DebugSegments=20,DebugThickness=0.25)
    DriverAxis=(X=-1,Y=0,Z=0)
    DriverItem=(Type=Bone,Name="clavicle_r")
    FalloffRegionScaleFactors=(PositiveWidth=-1,NegativeWidth=1,PositiveHeight=-1,NegativeHeight=1)
    FalloffSize=0.2
    FlipHeightScaling=False
    FlipWidthScaling=False
    OptionalParentItem=(Type=Bone,Name="None")
    RotationOffset=(X=110,Y=-15,Z=51)
   )
   > RigVMVariableNode(Value=@n8.OutputParam) {variable=Clavicle Back rt}
   > n9 = &RigUnit_SphericalPoseReader(
    ActiveRegionScaleFactors=(PositiveWidth=1,NegativeWidth=1,PositiveHeight=3,NegativeHeight=1)
    ActiveRegionSize=0.1
    Debug=(bDrawDebug=false,bDraw2D=False,bDrawLocalAxes=False,DebugScale=25,DebugSegments=20,DebugThickness=0.25)
    DriverAxis=(X=1,Y=0,Z=0)
    DriverItem=(Type=Bone,Name="clavicle_l")
    FalloffRegionScaleFactors=(PositiveWidth=1,NegativeWidth=1,PositiveHeight=-1,NegativeHeight=1)
    FalloffSize=0.3
    FlipHeightScaling=False
    FlipWidthScaling=False
    OptionalParentItem=(Type=Bone,Name="None")
    RotationOffset=(X=-70.457,Y=4.9077,Z=37.1574)
   )
   > RigVMVariableNode(Value=@n9.OutputParam) {variable=Clavicle Back lf}
   > n10 = &RigUnit_SphericalPoseReader(
    ActiveRegionScaleFactors=(PositiveWidth=1,NegativeWidth=1,PositiveHeight=1,NegativeHeight=1)
    ActiveRegionSize=0.15
    Debug=(bDrawDebug=false,bDraw2D=false,DebugScale=25,DebugSegments=20,DebugThickness=0.25)
    DriverAxis=(X=1,Y=0,Z=0)
    DriverItem=(Type=Bone,Name="head")
    FalloffRegionScaleFactors=(PositiveWidth=1,NegativeWidth=1,PositiveHeight=1,NegativeHeight=1)
    FalloffSize=0.32
    FlipHeightScaling=true
    FlipWidthScaling=false
    OptionalParentItem=(Type=Bone)
    RotationOffset=(X=0,Y=-30,Z=0)
   )
   > n11 = &RigUnit_SphericalPoseReader(
    ActiveRegionScaleFactors=(PositiveWidth=1,NegativeWidth=1,PositiveHeight=1,NegativeHeight=1)
    ActiveRegionSize=0.15
    Debug=(bDrawDebug=false,bDraw2D=false,DebugScale=25,DebugSegments=20,DebugThickness=0.25)
    DriverAxis=(X=1,Y=0,Z=0)
    DriverItem=(Type=Bone,Name="head")
    FalloffRegionScaleFactors=(PositiveWidth=1,NegativeWidth=1,PositiveHeight=1,NegativeHeight=1)
    FalloffSize=0.32
    FlipHeightScaling=true
    FlipWidthScaling=false
    OptionalParentItem=(Type=Bone)
    RotationOffset=(X=180,Y=30,Z=0)
   )
   &RigUnit_ModifyTransforms(
    ItemToModify=((Item=(Type=Bone,Name="head_aim_tgt"),Transform=(Rotation=(X=0,Y=0,Z=0,W=1),Translation=(X=-20,Y=-5,Z=0),Scale3D=(X=1,Y=1,Z=1))))
    WeightMinimum=0
    ExecutePin=@RigVMFunction_Sequence_2.B
    Weight=@n3.OutputParam
   )
   > &RigUnit_ModifyTransforms(
    ItemToModify=((Item=(Type=Bone,Name="head_aim_tgt"),Transform=(Rotation=(X=0,Y=0,Z=0,W=1),Translation=(X=-10,Y=-5,Z=0),Scale3D=(X=1,Y=1,Z=1))))
    WeightMinimum=0
    Weight=@n4.OutputParam
   )
   > &RigUnit_ModifyTransforms(
    ItemToModify=((Item=(Type=Bone,Name="head_aim_tgt"),Transform=(Rotation=(X=0,Y=0,Z=0,W=1),Translation=(X=-0.2512,Y=3.09,Z=0.0458),Scale3D=(X=1,Y=1,Z=1))))
    WeightMinimum=0
    Weight=@n5.OutputParam
   )
   > &RigUnit_ModifyTransforms(
    ItemToModify=((Item=(Type=Bone,Name="head_aim_tgt"),Transform=(Rotation=(X=0,Y=0,Z=0,W=1),Translation=(X=0,Y=0,Z=9),Scale3D=(X=1,Y=1,Z=1))))
    WeightMinimum=0
    Weight=@n6.OutputParam
   )
   > &RigUnit_ModifyTransforms(
    ItemToModify=((Item=(Type=Bone,Name="head_aim_tgt"),Transform=(Rotation=(X=0,Y=0,Z=0,W=1),Translation=(X=0,Y=0,Z=-9),Scale3D=(X=1,Y=1,Z=1))))
    WeightMinimum=0
    Weight=@n7.OutputParam
   )
   &RigUnit_AimConstraintLocalSpaceOffset(
    Child=(Type=Bone,Name="aim_hair_lf")
    Filter=(bX=true,bY=true,bZ=true)
    Parents=((Item=(Type=Bone,Name="head_aim_tgt"),Weight=1))
    UpAxis=(X=0,Y=-1,Z=0)
    Weight=1
    WorldUp=(Target=(X=0,Y=-1,Z=0),Kind=Location,Space=(Type=Bone,Name="head_aim_tgt"))
    bMaintainOffset=true
    ExecutePin=@RigVMFunction_Sequence_2.D
   )
   > &RigUnit_AimConstraintLocalSpaceOffset(
    Child=(Type=Bone,Name="aim_hair_cr")
    Filter=(bX=true,bY=true,bZ=true)
    Parents=((Item=(Type=Bone,Name="head_aim_tgt"),Weight=1))
    UpAxis=(X=0,Y=-1,Z=0)
    Weight=1
    WorldUp=(Target=(X=0,Y=-1,Z=0),Kind=Location,Space=(Type=Bone,Name="head_aim_tgt"))
    bMaintainOffset=true
   )
   > &RigUnit_AimConstraintLocalSpaceOffset(
    Child=(Type=Bone,Name="aim_hair_rt")
    Filter=(bX=true,bY=true,bZ=true)
    Parents=((Item=(Type=Bone,Name="head_aim_tgt"),Weight=1))
    UpAxis=(X=0,Y=-1,Z=0)
    Weight=1
    WorldUp=(Target=(X=0,Y=-1,Z=0),Kind=Location,Space=(Type=Bone,Name="head_aim_tgt"))
    bMaintainOffset=true
   )
   &RigUnit_ModifyTransforms(
    ItemToModify=((Item=(Type=Bone,Name="base_hair_rt"),Transform=(Rotation=(X=0,Y=0,Z=0,W=1),Translation=(X=-5,Y=0,Z=0),Scale3D=(X=1,Y=1,Z=1))),(Item=(Type=Bone,Name="base_hair_cr"),Transform=(Rotation=(X=0,Y=0,Z=0,W=1),Translation=(X=0,Y=0,Z=0),Scale3D=(X=1,Y=1,Z=1))))
    WeightMinimum=0
    ExecutePin=@RigVMFunction_Sequence_2.E
    Weight=@n3.OutputParam
   )
   > &RigUnit_ModifyTransforms(
    ItemToModify="((Item=(Type=Bone,Name=\"offset_hair_rt\"),Transform=(Rotation=(X=0.1188,Y=0.1116,Z=0.0113,W=0.9866),Translation=(X=2.3201,Y=0.3869,Z=3.3619),Scale3D=(X=1,Y=1,Z=1))),(Item=(Type=Bone,Name=\"offset_hair_cr\"),Transform=(Rotation=(X=0.0473,Y=0.0501,Z=-0.0715,W=0 ...(truncated)"
    WeightMinimum=0
    Weight=RigVMVariableNode() {variable=Head Left}.Value
   )
   > &RigUnit_ModifyTransforms(
    ItemToModify="((Item=(Type=Bone,Name=\"offset_hair_rt\"),Transform=(Rotation=(X=0,Y=0,Z=0.1414,W=0.9899),Translation=(X=0.9032,Y=-6.549,Z=2.1002),Scale3D=(X=1,Y=1,Z=1))),(Item=(Type=Bone,Name=\"offset_hair_cr\"),Transform=(Rotation=(X=-0.0062,Y=0.0581,Z=0.1061,W=0.9926),Tra ...(truncated)"
    WeightMinimum=0
    Weight=RigVMVariableNode() {variable=Head Right}.Value
   )
   > &RigUnit_ModifyTransforms(
    ItemToModify="((Item=(Type=Bone,Name=\"base_hair_rt\"),Transform=(Rotation=(X=0,Y=0,Z=0,W=1),Translation=(X=3,Y=0,Z=0),Scale3D=(X=1,Y=1,Z=1))),(Item=(Type=Bone,Name=\"base_hair_cr\"),Transform=(Rotation=(X=0,Y=0,Z=0,W=1),Translation=(X=7.2256,Y=0.1006,Z=2.0769),Scale3D=(X=1 ...(truncated)"
    WeightMinimum=0
    Weight=RigVMVariableNode() {variable=Head Down}.Value
   )
   > &RigUnit_ModifyTransforms(
    ItemToModify=((Item=(Type=Bone,Name="offset_hair_lf"),Transform=(Rotation=(X=-0.0494,Y=0.3384,Z=0.0541,W=-0.9381),Translation=(X=0,Y=-4,Z=0),Scale3D=(X=1,Y=1,Z=1))))
    WeightMinimum=0
    Weight=RigVMVariableNode() {variable=Clavicle Back lf}.Value
   )
   > &RigUnit_ModifyTransforms(
    ItemToModify=((Item=(Type=Bone,Name="offset_hair_rt"),Transform=(Rotation=(X=-0.0076,Y=-0.1735,Z=-0.043,W=0.9839),Translation=(X=0,Y=4,Z=0),Scale3D=(X=1,Y=1,Z=1))))
    WeightMinimum=0
    Weight=RigVMVariableNode() {variable=Clavicle Back rt}.Value
   )
   > &RigUnit_ModifyTransforms(
    ItemToModify=((Item=(Type=Bone,Name="head_aim_tgt"),Transform=(Rotation=(X=0,Y=0,Z=0,W=1),Translation=(X=0,Y=0,Z=7.883),Scale3D=(X=1,Y=1,Z=1))))
    WeightMinimum=0
    Weight=@n10.OutputParam
   )
   > &RigUnit_ModifyTransforms(
    ItemToModify=((Item=(Type=Bone,Name="head_aim_tgt"),Transform=(Rotation=(X=0,Y=0,Z=0,W=1),Translation=(X=0,Y=0,Z=-7.883),Scale3D=(X=1,Y=1,Z=1))))
    WeightMinimum=0
    Weight=@n11.OutputParam
   )
   RigVMFunctionReturnNode(
    Result=(X=0,Y=0,Z=0,W=1)
    Head Right=@n7.OutputParam
    Head Left=@n6.OutputParam
    Head Up=@n3.OutputParam
    Head Up OverShoot=@n4.OutputParam
    Head Down=@n5.OutputParam
    ExecuteContext=@n2.B
   )
  ]}
 RigVMFunctionLibrary {links [], nodes []}
 RigVMFunction_Sequence_3_ContainedGraph {nodes [
   n1 = RigVMFunctionEntryNode()
   n2 = &RigVMFunction_Sequence(ExecuteContext=@n1.ExecuteContext)
   n3 = &RigVMFunction_Sequence(ExecuteContext=@n2.B)
   n4 = &RigVMFunction_Sequence(ExecuteContext=@n3.B)
   n5 = &RigVMFunction_Sequence(ExecuteContext=@n4.B)
   RigVMFunctionReturnNode(B=@n3.A, A=@n2.A, E=@n5.A, F=@n5.B, D=@n4.A)
  ]}
 RigVMModel {nodes [
   RigUnit_BeginExecution = &RigUnit_BeginExecution()
   "HairSetup " = RigVMCollapseNode(ExecuteContext=@RigUnit_BeginExecution.ExecutePin)
  ]}
]
hierarchy [
 root {loc=<0 0 0>, rot=<0 0 0 1>}
 pelvis {parent=root, loc=<0 -0.5301 98.0571>, rot=<0.0127 0.707 -0.0127 -0.707>}
 spine_01 {parent=pelvis, loc=<6.1208 -0.9328 0>, rot=<0 0 0.034 -0.9994>}
 spine_02 {parent=spine_01, loc=<6.5997 -0.2095 0>, rot=<0 0 0.0238 -0.9997>}
 spine_03 {parent=spine_02, loc=<6.4914 -0.1237 0>, rot=<0 0 0.0149 -0.9999>}
 spine_04 {parent=spine_03, loc=<6.9598 0.0794 0>, rot=<0 0 -0.0132 -0.9999>}
 spine_05 {parent=spine_04, loc=<6.9987 0 0>, rot=<0 0 -0.0015 -1>}
 neck_01 {parent=spine_05, loc=<11.3138 0.1529 0>, rot=<0 0 -0.2003 -0.9797>}
 neck_02 {parent=neck_01, loc=<4.5627 0 0>, rot=<0 0 0.0504 -0.9987>}
 head {parent=neck_02, loc=<5.3914 0 0>, rot=<0 0 0.115 -0.9934>}
 base_hair_master_00 {parent=head, loc=<15.6759 -3.8498 0>, rot=<0.1086 -0.9941 0 0>}
 base_ribbon_a_00 {parent=base_hair_master_00, loc=<-2.2629 5.0624 10.2322>, rot=<-0.0109 0.364 0.857 -0.3644>}
 dyn_ribbon_a_00 {parent=base_ribbon_a_00, loc=<0 0 0>, rot=<0 0 0 1>}
 base_ribbon_b_00 {parent=base_hair_master_00, loc=<-1.7021 5.9438 10.3643>, rot=<0.3231 0.3304 0.3295 -0.8233>}
 dyn_ribbon_b_00 {parent=base_ribbon_b_00, loc=<0 0 0>, rot=<0 0 0 1>}
 base_hair_lf {parent=base_hair_master_00, loc=<6.0483 -6.3835 10.7372>, rot=<-0.7118 -0.0194 -0.0155 -0.7019>}
 aim_hair_lf {parent=base_hair_lf, loc=<0 0 0>, rot=<0 0 0 1>}
 offset_hair_lf {parent=aim_hair_lf, loc=<0 0 0>, rot=<0 0 0 1>}
 dyn_hair_01_lf {parent=offset_hair_lf, loc=<0 0 0>, rot=<0 0 0 1>}
 dyn_hair_02_lf {parent=dyn_hair_01_lf, loc=<13.5092 0 0>, rot=<0 0 0 1>}
 dyn_hair_03_lf {parent=dyn_hair_02_lf, loc=<7.0701 0 0>, rot=<0 0 0 1>}
 dyn_hair_04_lf {parent=dyn_hair_03_lf, loc=<7.0629 0 0>, rot=<0 0 -0.0506 -0.9987>}
 dyn_hair_05_lf {parent=dyn_hair_04_lf, loc=<17.6817 0 0>, rot=<0 0 0 1>}
 dyn_hair_06_lf {parent=dyn_hair_05_lf, loc=<16.6741 0 0>, rot=<0 0 0 1>}
 dyn_hair_07_lf {parent=dyn_hair_06_lf, loc=<14.9253 0 0>, rot=<0 0 0 1>}
 base_hair_cr {parent=base_hair_master_00, loc=<5.638 -9.1827 -0.2486>, rot=<-0.7172 -0.0077 -0.0036 -0.6968>}
 aim_hair_cr {parent=base_hair_cr, loc=<0 0 0>, rot=<0 0 0 1>}
 offset_hair_cr {parent=aim_hair_cr, loc=<0 0 0>, rot=<0 0 0 1>}
 dyn_hair_01_cr {parent=offset_hair_cr, loc=<0 0 0>, rot=<0 0 0 1>}
 dyn_hair_02_cr {parent=dyn_hair_01_cr, loc=<13.7912 0 0>, rot=<0 0 0 1>}
 dyn_hair_03_cr {parent=dyn_hair_02_cr, loc=<7.0701 0 0>, rot=<0 0 0 1>}
 dyn_hair_04_cr {parent=dyn_hair_03_cr, loc=<7.0629 0 0>, rot=<0 0 0 1>}
 dyn_hair_05_cr {parent=dyn_hair_04_cr, loc=<17.6817 0 0>, rot=<0 0 0 1>}
 dyn_hair_06_cr {parent=dyn_hair_05_cr, loc=<16.6741 0 0>, rot=<0 0 0 1>}
 dyn_hair_07_cr {parent=dyn_hair_06_cr, loc=<14.9253 0 0>, rot=<0 0 0 1>}
 base_hair_rt {parent=base_hair_master_00, loc=<6.0483 -6.3835 -10.7855>, rot=<-0.7118 -0.0194 -0.0155 -0.7019>}
 aim_hair_rt {parent=base_hair_rt, loc=<0 0 0>, rot=<0 0 0 1>}
 offset_hair_rt {parent=aim_hair_rt, loc=<0 0 0>, rot=<0 0 0 1>}
 dyn_hair_01_rt {parent=offset_hair_rt, loc=<0 0 0>, rot=<0 0 0 1>}
 dyn_hair_02_rt {parent=dyn_hair_01_rt, loc=<13.5092 0 0>, rot=<0 0 0 1>}
 dyn_hair_03_rt {parent=dyn_hair_02_rt, loc=<7.0701 0 0>, rot=<0 0 0 1>}
 dyn_hair_04_rt {parent=dyn_hair_03_rt, loc=<7.0629 0 0>, rot=<0 0 0.0506 -0.9987>}
 dyn_hair_05_rt {parent=dyn_hair_04_rt, loc=<17.6817 0 0>, rot=<0 0 0 1>}
 dyn_hair_06_rt {parent=dyn_hair_05_rt, loc=<16.6741 0 0>, rot=<0 0 0 1>}
 dyn_hair_07_rt {parent=dyn_hair_06_rt, loc=<14.9253 0 0>, rot=<0 0 0 1>}
 clavicle_l {parent=spine_05, loc=<5.4831 -0.1053 -3.2946>, rot=<0.0465 -0.7289 -0.009 -0.6829>}
 upperarm_l {parent=clavicle_l, loc=<9.8658 -1.2136 -0.4105>, rot=<0.0164 -0.3245 -0.0269 -0.9454>}
 clavicle_r {parent=spine_05, loc=<5.4831 -0.1053 3.2946>, rot=<0.7289 0.0465 -0.6829 0.009>}
 upperarm_r {parent=clavicle_r, loc=<-9.8658 1.2136 0.4105>, rot=<0.0164 -0.3245 -0.0269 -0.9454>}
 base_head_aim_tgt {parent=spine_05, loc=<9.003 -16.7403 0.2354>, rot=<0.0075 -0.0008 0.0004 -1>}
 head_aim_tgt {parent=base_head_aim_tgt, loc=<0 0 0>, rot=<0 0 0 1>}
 dyn_L_haori_side_mstr {parent=pelvis, loc=<5.6551 -9.8488 -13.3436>, rot=<0.039 -0.8852 0.4523 -0.1018>}
 dyn_L_haori_side_1 {parent=dyn_L_haori_side_mstr, loc=<0 0 0>, rot=<0 0 0 1>}
 dyn_L_haori_side_2 {parent=dyn_L_haori_side_1, loc=<17 0 0>, rot=<0 0 0 1>}
 dyn_L_haori_side_3 {parent=dyn_L_haori_side_2, loc=<17 0 0>, rot=<0 0 0 1>}
 dyn_C_haori_back_mstr {parent=pelvis, loc=<4.7337 -14.0213 0>, rot=<0.0674 -0.7039 0.7039 -0.0674>}
 dyn_C_haori_back_1 {parent=dyn_C_haori_back_mstr, loc=<0 0 0>, rot=<0 0 0 1>}
 dyn_C_haori_back_2 {parent=dyn_C_haori_back_1, loc=<17 0 0>, rot=<0 0 0 1>}
 dyn_C_haori_back_3 {parent=dyn_C_haori_back_2, loc=<17 0 0>, rot=<0 0 0 1>}
 dyn_R_haori_side_mstr {parent=pelvis, loc=<5.6547 -9.8488 13.3436>, rot=<0.1018 -0.4523 0.8852 -0.039>}
 dyn_R_haori_side_1 {parent=dyn_R_haori_side_mstr, loc=<0 0 0>, rot=<0 0 0 1>}
 dyn_R_haori_side_2 {parent=dyn_R_haori_side_1, loc=<17 0 0.0001>, rot=<0 0 0 1>}
 dyn_R_haori_side_3 {parent=dyn_R_haori_side_2, loc=<17 0 -0.0001>, rot=<0 0 0 1>}
]
variables [
 Clavicle Back lf {type=double, default=0}
 Clavicle Back rt {type=double, default=0}
 Head Down {type=double, default=0}
 Head Left {type=double, default=0}
 Head Right {type=double, default=0}
 Head Up {type=double, default=0}
 Head Up OverShoot {type=double, default=0}
 IsTargeting {type=bool, default=False}
 PelvisToGround {type=double, default=0}
 swingPickaxe {type=bool, default=False}
]
