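# Animation Blueprint for the AtticDrill Fin character's face accessory, driving secondary motion on hair and ribbon bones.
# Blends copied attached-parent and body-mesh poses into a Control Rig, then tilts hair while skydiving.
# Cached poses feed a backbling/cape-dependent choice of two RigidBody physics assets, ribbon AnimDynamics, then backbling/cape hair offsets.
# The event graph caches the body mesh, backbling flags, crouch-targeting state and gravity settings from a parameter asset.
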AnimBlueprint /BRCosmetics/Characters/Player/Female/Medium/Bodies/F_MED_AtticDrill_Fin/Meshes/Parts/F_MED_AtticDrill_Fin_FaceAcc_AnimBP
target_skeleton=&/F_MED_AtticDrill_Fin_FaceAcc_Skeleton
graphs [
 AnimGraph {nodes [
   Output Pose(Result=Use cached pose 'backpackCapeCache_1'()) {layerGroup=None}
   n1 = Layered blend per bone(
    BasePose=Copy Pose From Mesh() {bUseAttachedParent=true}
    BlendPoses_0=Copy Pose From Mesh(SourceMeshComponent=Get Body_Mesh()) {bUseAttachedParent=true}
   ) {layerSetup [{branchFilters [
       {boneName=clavicle_l}
       {boneName=clavicle_r}
       {boneName=dyn_L_haori_side_mstr}
       {boneName=dyn_R_haori_side_mstr}
       {boneName=dyn_C_haori_back_mstr}
       {boneName=None}
       {boneName=None}
       {boneName=None}
      ]}]}
   n2 = Control Rig!(Source=@n1, Alpha, swingPickaxe:bool=false, IsTargeting:bool=false) -> (Pose) {controlRigAssetReference {blueprintRigClass=&/F_MED_AtticDrill_Fin_FaceAcc_CtrlRig.F_MED_AtticDrill_Fin_FaceAcc_CtrlRig_C}}
   Save cached pose 'controlRigCache_1'(Pose=@n2) {cacheName=controlRigCache_1}
   n3 = RigidBody!(
    ComponentPose=Local To Component(LocalPose=Use cached pose 'PreDynamicsPose_1'())
    Alpha
    OverrideWorldGravity:Vector="0,0,-980"
    SimSpaceSettings:Sim Space Settings Structure=(WorldAlpha=1,VelocityScaleZ=1,DampingAlpha=0,MaxLinearVelocity=350,MaxAngularVelocity=4,MaxLinearAcceleration=1500,MaxAngularAcceleration=9,ExternalLinearDragV=(X=1.1,Y=1.1,Z=0),ExternalLinearVelocity=(X=0,Y=0,Z=0),ExternalAngularVelocity=(X=0,Y=0,Z=0))
   ) -> (Pose) {
    bClampLinearTranslationLimitToRefPose=true
    bOverrideWorldGravity=true
    overridePhysicsAsset=&/F_MED_AtticDrill_Fin_FaceAcc_Physics
    simulationSpace=BaseBoneSpace
    alphaBoolBlend {blendOption=Linear}
    baseBoneRef {boneName=neck_02}
    overrideWorldGravity {z=-980}
    simSpaceSettings {
     dampingAlpha=0
     maxAngularAcceleration=9
     maxAngularVelocity=4
     maxLinearAcceleration=1500
     maxLinearVelocity=350
     worldAlpha=1
     externalLinearDragV=<1.1 1.1>
    }
   }
   n4 = Component To Local(ComponentPose=@n3)
   n5 = RigidBody(
    ComponentPose=Local To Component(LocalPose=Use cached pose 'PreDynamicsPose_1'())
    OverrideWorldGravity="0,0,-980"
    SimSpaceSettings=(WorldAlpha=1,VelocityScaleZ=1,DampingAlpha=0,MaxLinearVelocity=350,MaxAngularVelocity=4,MaxLinearAcceleration=1500,MaxAngularAcceleration=9,ExternalLinearDragV=(X=1.1,Y=1.1,Z=0),ExternalLinearVelocity=(X=0,Y=0,Z=0),ExternalAngularVelocity=(X=0,Y=0,Z=0))
   ) {
    bClampLinearTranslationLimitToRefPose=true
    bOverrideWorldGravity=true
    overridePhysicsAsset=&/F_MED_AtticDrill_Fin_FaceAcc_BackPack_Physics
    simulationSpace=BaseBoneSpace
    alphaBoolBlend {blendOption=Linear}
    baseBoneRef {boneName=neck_02}
    overrideWorldGravity {z=-980}
    simSpaceSettings {
     dampingAlpha=0
     maxAngularAcceleration=9
     maxAngularVelocity=4
     maxLinearAcceleration=1500
     maxLinearVelocity=350
     worldAlpha=1
     externalLinearDragV=<1.1 1.1>
    }
   }
   n6 = Component To Local(ComponentPose=@n5)
   n7 = Get is Wearing cape and backpack() -> (is Wearing cape and backpack:bool)
   n8 = Anim Dynamics - Bone: dyn_ribbon_b_00(
    ComponentPose=Local To Component(LocalPose=Blend Poses by bool(BlendPose_0=@n6, BlendPose_1=@n4, bActiveValue=@n7) {blendType=Linear})
    Alpha=1
    ExternalForce="0,0,0"
   ) {
    angularDampingOverride=1
    angularSpringConstant=1000
    bAngularSpring=true
    bOverrideAngularDamping=true
    bOverrideLinearDamping=true
    bUseGravityOverride=true
    linearDampingOverride=1
    simulationSpace=RootRelative
    boundBone {boneName=dyn_ribbon_b_00}
    componentAppliedLinearAccClamp=<4000 4000 4000>
    gravityOverride {z=-800}
    physicsBodyDefinitions [{
      boundBone {boneName=dyn_ribbon_b_00}
      boxExtents=<11 11 11>
      constraintSetup {angularLimitsMax=<5 20 20>, angularLimitsMin=<-5 -20 -20>, angularTarget=<1 0 0>}
      localJointOffset=<2 0 0>
     }]
    simSpaceSettings {maxAngularAcceleration=40, maxAngularVelocity=8, simSpaceAngularAlpha=1}
   }
   n9 = Anim Dynamics - Bone: dyn_ribbon_a_00(ComponentPose=@n8, Alpha=1, ExternalForce="0,0,0") {
    angularDampingOverride=1
    angularSpringConstant=1000
    bAngularSpring=true
    bOverrideAngularDamping=true
    bOverrideLinearDamping=true
    bUseGravityOverride=true
    linearDampingOverride=1
    simulationSpace=RootRelative
    boundBone {boneName=dyn_ribbon_a_00}
    componentAppliedLinearAccClamp=<4000 4000 4000>
    gravityOverride {z=-800}
    physicsBodyDefinitions [{
      boundBone {boneName=dyn_ribbon_a_00}
      boxExtents=<11 11 11>
      constraintSetup {angularLimitsMax=<5 20 20>, angularLimitsMin=<-5 -20 -20>, angularTarget=<1 0 0>}
      localJointOffset=<2 0 0>
     }]
    simSpaceSettings {maxAngularAcceleration=40, maxAngularVelocity=8, simSpaceAngularAlpha=1}
   }
   Save cached pose 'dynamicsCache_1'(Pose=Component To Local(ComponentPose=@n9)) {cacheName=dynamicsCache_1}
   n10 = Use cached pose 'controlRigCache_1'()
   n11 = Get bIsSkyDiving()
   n12 = "Transform (Modify) Bone - Bone: dyn_hair_01_cr"!(
    ComponentPose="Transform (Modify) Bone - Bone: dyn_hair_01_lf"!(ComponentPose=Local To Component(LocalPose=@n10), bAlphaBoolEnabled:bool=@n11, Translation="0,0,0", Rotation="0,0,15", Scale="1,1,1") -> (Pose) {alphaInputType=Bool, rotationMode=BMM_Additive, boneToModify {boneName=dyn_hair_01_lf}, rotation {roll=15}}
    bAlphaBoolEnabled:bool=@n11
    Translation="0,0,0"
    Rotation="0,0,15"
    Scale="1,1,1"
   ) -> (Pose) {alphaInputType=Bool, rotationMode=BMM_Additive, boneToModify {boneName=dyn_hair_01_cr}, rotation {roll=15}}
   Save cached pose 'PreDynamicsPose_1'(Pose=Component To Local(ComponentPose="Transform (Modify) Bone - Bone: dyn_hair_01_rt"!(ComponentPose=@n12, bAlphaBoolEnabled:bool=@n11, Translation="0,0,0", Rotation="0,0,15", Scale="1,1,1") -> (Pose) {alphaInputType=Bool, rotationMode=BMM_Additive, boneToModify {boneName=dyn_hair_01_rt}, rotation {roll=15}})) {cacheName=PreDynamicsPose_1}
   n13 = Local To Component(LocalPose=Use cached pose 'PostDynCache_1_1'())
   Save cached pose 'backpackCapeCache_1'(Pose=Blend Poses by bool(
     BlendPose_0=Component To Local(ComponentPose=@n13)
     BlendPose_1=Use cached pose 'PostDynCache_1_1'()
     bActiveValue=Get is Wearing cape and backpack()
    ) {blendType=Linear}) {cacheName=backpackCapeCache_1}
   n14 = Get bIsSliding()
   n15 = "Transform (Modify) Bone - Bone: dyn_hair_05_lf"!(
    ComponentPose=Local To Component(LocalPose=Use cached pose 'dynamicsCache_1'())
    bAlphaBoolEnabled:bool=@n14
    Translation="0,0,0"
    Rotation="0,0,0"
    Scale="1,1,1"
   ) -> (Pose) {alphaInputType=Bool, rotationMode=BMM_Additive, boneToModify {boneName=dyn_hair_05_lf}}
   n16 = Get is Wearing cape and backpack()
   n17 = "Transform (Modify) Bone - Bone: dyn_hair_02_lf"!(
    ComponentPose="Transform (Modify) Bone - Bone: dyn_hair_05_rt"!(
     ComponentPose="Transform (Modify) Bone - Bone: dyn_hair_05_cr"!(ComponentPose=@n15, bAlphaBoolEnabled:bool=@n14, Translation="0,0,0", Rotation="0,0,0", Scale="1,1,1") -> (Pose) {alphaInputType=Bool, rotationMode=BMM_Additive, boneToModify {boneName=dyn_hair_05_cr}}
     bAlphaBoolEnabled:bool=@n14
     Translation="0,0,0"
     Rotation="0,0,0"
     Scale="1,1,1"
    ) -> (Pose) {alphaInputType=Bool, rotationMode=BMM_Additive, boneToModify {boneName=dyn_hair_05_rt}}
    bAlphaBoolEnabled:bool=@n16
    Translation="0,0,-2"
    Rotation="0,0,0"
    Scale="1,1,0.2"
   ) -> (Pose) {
    alphaInputType=Bool
    scaleMode=BMM_Additive
    scaleSpace=BCS_BoneSpace
    translationMode=BMM_Additive
    translationSpace=BCS_ParentBoneSpace
    boneToModify {boneName=dyn_hair_02_lf}
    scale {z=0.2}
    translation {z=-2}
   }
   n18 = "Transform (Modify) Bone - Bone: dyn_hair_02_cr"!(ComponentPose=@n17, bAlphaBoolEnabled:bool=@n16, Translation="0,0,-2", Rotation="0,0,0", Scale="1,1,0.2") -> (Pose) {
    alphaInputType=Bool
    scaleMode=BMM_Additive
    scaleSpace=BCS_BoneSpace
    translationMode=BMM_Additive
    translationSpace=BCS_ParentBoneSpace
    boneToModify {boneName=dyn_hair_02_cr}
    scale {z=0.2}
    translation {z=-2}
   }
   n19 = "Transform (Modify) Bone - Bone: dyn_hair_02_rt"!(ComponentPose=@n18, bAlphaBoolEnabled:bool=@n16, Translation="0,0,-2", Rotation="0,0,0", Scale="1,1,0.2") -> (Pose) {
    alphaInputType=Bool
    scaleMode=BMM_Additive
    scaleSpace=BCS_BoneSpace
    translationMode=BMM_Additive
    translationSpace=BCS_ParentBoneSpace
    boneToModify {boneName=dyn_hair_02_rt}
    scale {z=0.2}
    translation {z=-2}
   }
   Save cached pose 'PostDynCache_1_1'(Pose=Component To Local(ComponentPose=@n19)) {cacheName=PostDynCache_1_1}
  ]}
 EventGraph {nodes [
   Event BlueprintUpdateAnimation()
   > n1 = Sequence!(execute) -> (then_0, then_1, then_2:Exec)
   > n2 = "Gravity Override State Function (Parameter Asset) (Nativized)"(
    CharacterPartAnimInstance=Self-Reference()
    GameTimeInSeconds=GetGameTimeInSeconds()
    Parameters=&/F_MED_AtticDrill_Fin_FaceAcc_Parameters
   )
   > Set GravityOverride(GravityOverride=@n2)
   > Set RBAN Settings Out(RBAN Settings Out=@n2.OutRigidBodyAnimNodeParameters)
   Set IsCrouchTarget(execute=@n1.then_1, IsCrouchTarget:bool=AND Boolean(A=Get bIsCrouching(), B=Get bIsTargeting())) -> (Output_Get:bool)
   Event OnCharacterCustomizationLoaded()
   > n3 = Cast To PlayerPawn_Athena(Object=TryGetPawnOwner())
   > Set Body_Mesh(Body_Mesh=GetSkeletalMeshForPartType(self=@n3))
   n4 = CheckForBackbling()
   > n5 = Set isWearingBackbling(isWearingBackbling=AND Boolean!(A=@n4, B=NOT Boolean(A=@n4.bIsBackblingACape), C:bool=NOT Boolean(A=@n4.bIsBackblingLong)) -> (ReturnValue))
   > n6 = Set Is Wearing Cape(Is Wearing Cape=OR Boolean(A=@n4.bIsBackblingACape, B=@n4.bIsBackblingLong))
   > Set is Wearing cape and backpack(is Wearing cape and backpack:bool=OR Boolean(A=@n5, B=@n6)) -> (Output_Get:bool)
  ]}
]
variables [
 "Body Mesh " {type=SkeletalMeshComponent}
 Body_Mesh {type=SkeletalMeshComponent}
 GravityOverride {type=Vector}
 Is Wearing Cape {type=boolean}
 is Wearing cape and backpack {type=boolean}
 IsCrouchTarget {type=boolean}
 isWearingBackbling {type=boolean}
 RBAN Settings Out {type=FortRigidBodyAnimNodeParameters}
]
