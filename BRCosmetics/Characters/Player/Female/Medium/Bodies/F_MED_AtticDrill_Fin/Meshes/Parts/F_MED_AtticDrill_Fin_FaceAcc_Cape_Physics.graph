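# Physics asset for the AtticDrill Fin female character's face-accessory cape, built from capsule bodies on dyn_hair, dyn_haori and body bones.
# Left, center and right dyn_hair chains are rigidly anchored to spine_04 and swing within limits, with twist locked.
# Angularly free constraints cross-link the center chain's tip to both side tips.
# Haori, spine and upper-arm capsules use fully locked angular constraints leading back to the pelvis.
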PhysicsAsset /BRCosmetics/Characters/Player/Female/Medium/Bodies/F_MED_AtticDrill_Fin/Meshes/Parts/F_MED_AtticDrill_Fin_FaceAcc_Cape_Physics
bodies {
 dyn_C_haori_back_1 [{shapeType=Capsule, length=17.0144, radius=13.0084, shapeName=dyn_C_haori_back_1_capsule, center=<16.8796 0 -10.6216>, rotation=<rot 90 0 0>}]
 dyn_C_haori_back_3 [{
   shapeType=Capsule
   length=12.1426
   radius=13.0084
   shapeName=dyn_C_haori_back_3_capsule
   center=<8.732 0 -9.9485>
   rotation=<rot 83.5671 180 180>
  }]
 dyn_L_haori_side_1 [{
   shapeType=Capsule
   length=17.0144
   radius=13.0084
   shapeName=dyn_L_haori_side_1_capsule
   center=<16.6279 2.3849 -7.6957>
   rotation=<rot 90 6.0237 0>
  }]
 dyn_L_haori_side_3 [{
   shapeType=Capsule
   length=12.1426
   radius=13.0084
   shapeName=dyn_L_haori_side_3_capsule
   center=<8.2248 5.8289 -5.5083>
   rotation=<rot 82.4683 -173.5827 -168.9813>
  }]
 dyn_R_haori_side_1 [{
   shapeType=Capsule
   length=17.0144
   radius=13.0084
   shapeName=dyn_R_haori_side_1_capsule
   center=<16.7966 -1.2628 -7.6957>
   rotation=<rot 90 -3.76 0>
  }]
 dyn_R_haori_side_3 [{
   shapeType=Capsule
   length=12.1426
   radius=13.0084
   shapeName=dyn_R_haori_side_3_capsule
   center=<8.087 -5.5502 -5.6852>
   rotation=<rot 80.6654 162.9254 173.1986>
  }]
 dyn_hair_03_cr [{shapeType=Capsule, length=2, radius=2, shapeName=dyn_hair_03_cr_capsule}]
 dyn_hair_03_lf [{shapeType=Capsule, length=2, radius=2, shapeName=dyn_hair_03_lf_capsule}]
 dyn_hair_03_rt [{shapeType=Capsule, length=2, radius=2, shapeName=dyn_hair_03_rt_capsule}]
 dyn_hair_04_cr [{shapeType=Capsule, length=1.224, radius=8.776, shapeName=dyn_hair_04_cr_capsule, center=<9.8747 0 0>, rotation=<rot 90 0 0>}]
 dyn_hair_04_lf [{shapeType=Capsule, length=13.4404, radius=7.0564, shapeName=dyn_hair_04_lf_capsule, center=<9.1143 4.6873 2.2481>, rotation=<rot 90 -90 0>}]
 dyn_hair_04_rt [{shapeType=Capsule, length=13.4404, radius=7.0564, shapeName=dyn_hair_04_rt_capsule, center=<9.1143 -4.6873 2.2481>, rotation=<rot 90 -90 0>}]
 dyn_hair_05_cr [{shapeType=Capsule, length=4.1924, radius=7.6206, shapeName=dyn_hair_05_cr_capsule, center=<9.4242 0 0>, rotation=<rot 90 0 0>}]
 dyn_hair_05_lf [{shapeType=Capsule, length=16.0246, radius=6.7161, shapeName=dyn_hair_05_lf_capsule, center=<7.5681 2.2 3.1>, rotation=<rot 90 -90 0>}]
 dyn_hair_05_rt [{shapeType=Capsule, length=16.0246, radius=6.7161, shapeName=dyn_hair_05_rt_capsule, center=<7.5681 -2.2293 3.1396>, rotation=<rot 90 -90 0>}]
 dyn_hair_06_cr [{shapeType=Capsule, length=6.9108, radius=6.8196, shapeName=dyn_hair_06_cr_capsule, center=<8.6462 0 0>, rotation=<rot 90 0 0>}]
 dyn_hair_06_lf [{
   shapeType=Capsule
   length=19.1587
   radius=5.8374
   shapeName=dyn_hair_06_lf_capsule
   center=<7.5681 0 4.2>
   rotation=<rot 79.653 89.9993 179.9993>
  }]
 dyn_hair_06_rt [{
   shapeType=Capsule
   length=19.1587
   radius=5.8374
   shapeName=dyn_hair_06_rt_capsule
   center=<7.5681 0 4.1933>
   rotation=<rot -79.653 89.9993 179.9993>
  }]
 dyn_hair_07_cr [{shapeType=Capsule, length=3.388, radius=9.1026, shapeName=dyn_hair_07_cr_capsule, center=<12.7859 0 0>, rotation=<rot 90 0 0>}]
 dyn_hair_07_lf [{
   shapeType=Capsule
   length=17.9651
   radius=9.1013
   shapeName=dyn_hair_07_lf_capsule
   center=<14.8079 0.7499 3.5>
   rotation=<rot 77.8868 89.9994 0.0006>
  }]
 dyn_hair_07_rt [{
   shapeType=Capsule
   length=17.9651
   radius=9.1013
   shapeName=dyn_hair_07_rt_capsule
   center=<13.5385 -0.7567 3.5255>
   rotation=<rot 77.8868 -89.9995 0.0006>
  }]
 pelvis [{shapeType=Capsule, length=0.505, radius=0.505, shapeName=pelvis_capsule, rotation=<rot 0 -90 0>}]
 spine_02 [{shapeType=Capsule, length=8.182, radius=14.9513, shapeName=spine_02_capsule, center=<3.0082 0.858 0>}]
 spine_04 [{shapeType=Capsule, length=9.835, radius=13.2982, shapeName=spine_04_capsule, center=<5.3903 2.7307 0>}]
 upperarm_l [{shapeType=Capsule, length=26.5778, radius=8.005, shapeName=upperarm_l_capsule, center=<20.3766 0.1906 -2.2777>, rotation=<rot -90 0 0>}]
 upperarm_r [{shapeType=Capsule, length=26.5778, radius=8.005, shapeName=upperarm_r_capsule, center=<-20.3766 -0.1906 2.2777>, rotation=<rot -90 180 0>}]
}
constraints [
 {bone1Name=dyn_hair_04_rt, bone2Name=dyn_hair_03_rt, swing1LimitDegrees=20, swing2LimitDegrees=45, twistLimitDegrees=45, twistMotion=Locked}
 {bone1Name=dyn_hair_05_rt, bone2Name=dyn_hair_04_rt, swing1LimitDegrees=20, swing2LimitDegrees=30, twistLimitDegrees=45, twistMotion=Locked}
 {bone1Name=dyn_hair_06_rt, bone2Name=dyn_hair_05_rt, swing1LimitDegrees=20, swing2LimitDegrees=20, twistLimitDegrees=45, twistMotion=Locked}
 {bone1Name=dyn_hair_07_rt, bone2Name=dyn_hair_06_rt, swing1LimitDegrees=10, swing2LimitDegrees=25, twistLimitDegrees=45, twistMotion=Locked}
 {bone1Name=dyn_hair_04_lf, bone2Name=dyn_hair_03_lf, swing1LimitDegrees=20, swing2LimitDegrees=45, twistLimitDegrees=45, twistMotion=Locked}
 {bone1Name=dyn_hair_05_lf, bone2Name=dyn_hair_04_lf, swing1LimitDegrees=20, swing2LimitDegrees=30, twistLimitDegrees=45, twistMotion=Locked}
 {bone1Name=dyn_hair_06_lf, bone2Name=dyn_hair_05_lf, swing1LimitDegrees=20, swing2LimitDegrees=20, twistLimitDegrees=45, twistMotion=Locked}
 {bone1Name=dyn_hair_07_lf, bone2Name=dyn_hair_06_lf, swing1LimitDegrees=10, swing2LimitDegrees=25, twistLimitDegrees=45, twistMotion=Locked}
 {bone1Name=dyn_hair_04_cr, bone2Name=dyn_hair_03_cr, swing1LimitDegrees=20, swing2LimitDegrees=45, twistLimitDegrees=45, twistMotion=Locked}
 {bone1Name=dyn_hair_05_cr, bone2Name=dyn_hair_04_cr, swing1LimitDegrees=20, swing2LimitDegrees=20, twistLimitDegrees=45, twistMotion=Locked}
 {bone1Name=dyn_hair_06_cr, bone2Name=dyn_hair_05_cr, swing1LimitDegrees=5, swing2LimitDegrees=20, twistLimitDegrees=45, twistMotion=Locked}
 {bone1Name=dyn_hair_07_cr, bone2Name=dyn_hair_06_cr, swing1LimitDegrees=5, swing2LimitDegrees=25, twistLimitDegrees=45, twistMotion=Locked}
 {
  bone1Name=dyn_hair_03_cr
  bone2Name=spine_04
  swing1LimitDegrees=45
  swing1Motion=Locked
  swing2LimitDegrees=45
  swing2Motion=Locked
  twistLimitDegrees=45
  twistMotion=Locked
 }
 {
  bone1Name=dyn_hair_03_lf
  bone2Name=spine_04
  swing1LimitDegrees=45
  swing1Motion=Locked
  swing2LimitDegrees=45
  swing2Motion=Locked
  twistLimitDegrees=45
  twistMotion=Locked
 }
 {
  bone1Name=dyn_hair_03_rt
  bone2Name=spine_04
  swing1LimitDegrees=45
  swing1Motion=Locked
  swing2LimitDegrees=45
  swing2Motion=Locked
  twistLimitDegrees=45
  twistMotion=Locked
 }
 {
  bone1Name=dyn_R_haori_side_1
  bone2Name=pelvis
  swing1LimitDegrees=45
  swing1Motion=Locked
  swing2LimitDegrees=45
  swing2Motion=Locked
  twistLimitDegrees=45
  twistMotion=Locked
 }
 {
  bone1Name=dyn_R_haori_side_3
  bone2Name=dyn_R_haori_side_1
  swing1LimitDegrees=45
  swing1Motion=Locked
  swing2LimitDegrees=45
  swing2Motion=Locked
  twistLimitDegrees=45
  twistMotion=Locked
 }
 {
  bone1Name=dyn_L_haori_side_3
  bone2Name=dyn_L_haori_side_1
  swing1LimitDegrees=45
  swing1Motion=Locked
  swing2LimitDegrees=45
  swing2Motion=Locked
  twistLimitDegrees=45
  twistMotion=Locked
 }
 {
  bone1Name=dyn_L_haori_side_1
  bone2Name=pelvis
  swing1LimitDegrees=45
  swing1Motion=Locked
  swing2LimitDegrees=45
  swing2Motion=Locked
  twistLimitDegrees=45
  twistMotion=Locked
 }
 {
  bone1Name=dyn_C_haori_back_3
  bone2Name=dyn_C_haori_back_1
  swing1LimitDegrees=45
  swing1Motion=Locked
  swing2LimitDegrees=45
  swing2Motion=Locked
  twistLimitDegrees=45
  twistMotion=Locked
 }
 {
  bone1Name=dyn_C_haori_back_1
  bone2Name=pelvis
  swing1LimitDegrees=45
  swing1Motion=Locked
  swing2LimitDegrees=45
  swing2Motion=Locked
  twistLimitDegrees=45
  twistMotion=Locked
 }
 {
  bone1Name=upperarm_l
  bone2Name=spine_04
  swing1LimitDegrees=45
  swing1Motion=Locked
  swing2LimitDegrees=45
  swing2Motion=Locked
  twistLimitDegrees=45
  twistMotion=Locked
 }
 {
  bone1Name=upperarm_r
  bone2Name=spine_04
  swing1LimitDegrees=45
  swing1Motion=Locked
  swing2LimitDegrees=45
  swing2Motion=Locked
  twistLimitDegrees=45
  twistMotion=Locked
 }
 {
  bone1Name=spine_02
  bone2Name=pelvis
  swing1LimitDegrees=45
  swing1Motion=Locked
  swing2LimitDegrees=45
  swing2Motion=Locked
  twistLimitDegrees=45
  twistMotion=Locked
 }
 {
  bone1Name=spine_04
  bone2Name=spine_02
  swing1LimitDegrees=45
  swing1Motion=Locked
  swing2LimitDegrees=45
  swing2Motion=Locked
  twistLimitDegrees=45
  twistMotion=Locked
 }
 {
  bone1Name=dyn_hair_07_cr
  bone2Name=dyn_hair_07_rt
  swing1LimitDegrees=45
  swing1Motion=Free
  swing2LimitDegrees=45
  swing2Motion=Free
  twistLimitDegrees=45
  twistMotion=Free
 }
 {
  bone1Name=dyn_hair_07_cr
  bone2Name=dyn_hair_07_lf
  swing1LimitDegrees=45
  swing1Motion=Free
  swing2LimitDegrees=45
  swing2Motion=Free
  twistLimitDegrees=45
  twistMotion=Free
 }
]
